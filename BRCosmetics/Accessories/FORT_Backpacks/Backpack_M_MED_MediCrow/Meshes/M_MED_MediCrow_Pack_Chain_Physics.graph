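# Physics asset for the MediCrow backpack cosmetic, defining capsule bodies and joints for its dangling chains and cape spine.
# Left and right six-link chains hang from a large `hider` capsule with moderate swing and twist limits.
# Their last links join through an angularly free constraint, closing the chains into a loop.
# A four-segment cape spine, also anchored to `hider`, uses much tighter limits, allowing only slight bending.
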PhysicsAsset /BRCosmetics/Accessories/FORT_Backpacks/Backpack_M_MED_MediCrow/Meshes/M_MED_MediCrow_Pack_Chain_Physics
bodies {
 dyn_cape_spine [{
   shapeType=Capsule
   length=7.5462
   radius=3.9136
   shapeName=dyn_cape_spine_capsule
   center=<1.0316 0.0001 0.2451>
   rotation=<rot 0 0.0051 -89.9961>
  }]
 dyn_cape_spine1 [{
   shapeType=Capsule
   length=7.0845
   radius=3.2553
   shapeName=dyn_cape_spine1_capsule
   center=<1.1502 -0.0016 0.5776>
   rotation=<rot 0 0.0517 -89.9998>
  }]
 dyn_cape_spine2 [{
   shapeType=Capsule
   length=5.8493
   radius=2.8197
   shapeName=dyn_cape_spine2_capsule
   center=<0.8905 -0.0014 -0.3904>
   rotation=<rot 0 -179.9538 -89.9987>
  }]
 dyn_cape_spine3 [{shapeType=Capsule, length=4.623, radius=2.2234, shapeName=dyn_cape_spine3_capsule, center=<2.3864 0 -0.3003>, rotation=<rot 0 90 -87.152>}]
 dyn_l_newchain_joint1 [{
   shapeType=Capsule
   length=1.9461
   radius=1.5951
   shapeName=dyn_l_newchain_joint1_capsule
   center=<4.5645 -0.0282 0.1605>
   rotation=<rot -0.9868 -89.9998 -88.7955>
  }]
 dyn_l_newchain_joint2 [{
   shapeType=Capsule
   length=2.805
   radius=1.6482
   shapeName=dyn_l_newchain_joint2_capsule
   center=<2.7549 -0.0052 0.0392>
   rotation=<rot 0 -89.3511 -87.249>
  }]
 dyn_l_newchain_joint3 [{
   shapeType=Capsule
   length=2.8227
   radius=1.8832
   shapeName=dyn_l_newchain_joint3_capsule
   center=<3.2558 0.084 -0.0315>
   rotation=<rot 0 -90.6841 -76.1371>
  }]
 dyn_l_newchain_joint4 [{
   shapeType=Capsule
   length=2.4505
   radius=1.6905
   shapeName=dyn_l_newchain_joint4_capsule
   center=<2.8623 0.0744 0.137>
   rotation=<rot 0 -89.4842 -79.4165>
  }]
 dyn_l_newchain_joint5 [{
   shapeType=Capsule
   length=2.6205
   radius=1.7628
   shapeName=dyn_l_newchain_joint5_capsule
   center=<2.539 -0.139 0.5188>
   rotation=<rot 0 95.5326 -88.9219>
  }]
 dyn_l_newchain_joint6 [{
   shapeType=Capsule
   length=2.5082
   radius=1.6463
   shapeName=dyn_l_newchain_joint6_capsule
   center=<2.3371 -0.2523 -0.0924>
   rotation=<rot 0 -88.934 -64.4386>
  }]
 dyn_r_newchain_joint1 [{
   shapeType=Capsule
   length=1.9461
   radius=1.5951
   shapeName=dyn_r_newchain_joint1_capsule
   center=<4.5645 -0.0282 -0.1604>
   rotation=<rot 0.9854 -89.9753 90.1953>
  }]
 dyn_r_newchain_joint2 [{
   shapeType=Capsule
   length=2.805
   radius=1.6482
   shapeName=dyn_r_newchain_joint2_capsule
   center=<2.755 -0.0052 -0.0392>
   rotation=<rot 0 90.6489 -87.249>
  }]
 dyn_r_newchain_joint3 [{
   shapeType=Capsule
   length=2.8227
   radius=1.8832
   shapeName=dyn_r_newchain_joint3_capsule
   center=<3.2558 0.084 0.0315>
   rotation=<rot 0 89.3157 -76.1369>
  }]
 dyn_r_newchain_joint4 [{
   shapeType=Capsule
   length=2.4505
   radius=1.6905
   shapeName=dyn_r_newchain_joint4_capsule
   center=<2.8622 0.0744 -0.137>
   rotation=<rot 0 90.516 -79.4168>
  }]
 dyn_r_newchain_joint5 [{
   shapeType=Capsule
   length=2.6205
   radius=1.7628
   shapeName=dyn_r_newchain_joint5_capsule
   center=<2.539 -0.139 -0.5188>
   rotation=<rot 0 -84.4675 -88.9219>
  }]
 dyn_r_newchain_joint6 [{
   shapeType=Capsule
   length=2.5924
   radius=1.7417
   shapeName=dyn_r_newchain_joint6_capsule
   center=<2.405 -0.3665 0.11>
   rotation=<rot 0 81.6784 -68.7645>
  }]
 hider [{shapeType=Capsule, length=27.7094, radius=20.6831, center=<1.0744 26.8878 -24.6886>, rotation=<rot 2.8242 -90.2454 -88.0328>}]
}
constraints [
 {bone1Name=dyn_l_newchain_joint1, bone2Name=hider, swing1LimitDegrees=30, swing2LimitDegrees=10, twistLimitDegrees=15}
 {bone1Name=dyn_l_newchain_joint2, bone2Name=dyn_l_newchain_joint1, swing1LimitDegrees=30, swing2LimitDegrees=10, twistLimitDegrees=15}
 {bone1Name=dyn_l_newchain_joint3, bone2Name=dyn_l_newchain_joint2, swing1LimitDegrees=30, swing2LimitDegrees=10, twistLimitDegrees=15}
 {bone1Name=dyn_l_newchain_joint4, bone2Name=dyn_l_newchain_joint3, swing1LimitDegrees=30, swing2LimitDegrees=10, twistLimitDegrees=15}
 {bone1Name=dyn_l_newchain_joint5, bone2Name=dyn_l_newchain_joint4, swing1LimitDegrees=30, swing2LimitDegrees=10, twistLimitDegrees=15}
 {bone1Name=dyn_l_newchain_joint6, bone2Name=dyn_l_newchain_joint5, swing1LimitDegrees=30, swing2LimitDegrees=10, twistLimitDegrees=15}
 {bone1Name=dyn_r_newchain_joint1, bone2Name=hider, swing1LimitDegrees=30, swing2LimitDegrees=10, twistLimitDegrees=15}
 {bone1Name=dyn_r_newchain_joint2, bone2Name=dyn_r_newchain_joint1, swing1LimitDegrees=30, swing2LimitDegrees=10, twistLimitDegrees=15}
 {bone1Name=dyn_r_newchain_joint3, bone2Name=dyn_r_newchain_joint2, swing1LimitDegrees=30, swing2LimitDegrees=10, twistLimitDegrees=15}
 {bone1Name=dyn_r_newchain_joint4, bone2Name=dyn_r_newchain_joint3, swing1LimitDegrees=30, swing2LimitDegrees=10, twistLimitDegrees=15}
 {bone1Name=dyn_r_newchain_joint5, bone2Name=dyn_r_newchain_joint4, swing1LimitDegrees=30, swing2LimitDegrees=10, twistLimitDegrees=15}
 {bone1Name=dyn_r_newchain_joint6, bone2Name=dyn_r_newchain_joint5, swing1LimitDegrees=30, swing2LimitDegrees=10, twistLimitDegrees=15}
 {
  bone1Name=dyn_r_newchain_joint6
  bone2Name=dyn_l_newchain_joint6
  swing1LimitDegrees=45
  swing1Motion=Free
  swing2LimitDegrees=45
  swing2Motion=Free
  twistLimitDegrees=45
  twistMotion=Free
 }
 {bone1Name=dyn_cape_spine, bone2Name=hider, swing1LimitDegrees=3, swing2LimitDegrees=5, twistLimitDegrees=1}
 {bone1Name=dyn_cape_spine1, bone2Name=dyn_cape_spine, swing1LimitDegrees=3, swing2LimitDegrees=5, twistLimitDegrees=1}
 {bone1Name=dyn_cape_spine2, bone2Name=dyn_cape_spine1, swing1LimitDegrees=3, swing2LimitDegrees=5, twistLimitDegrees=1}
 {bone1Name=dyn_cape_spine3, bone2Name=dyn_cape_spine2, swing1LimitDegrees=3, swing2LimitDegrees=5, twistLimitDegrees=1}
]
