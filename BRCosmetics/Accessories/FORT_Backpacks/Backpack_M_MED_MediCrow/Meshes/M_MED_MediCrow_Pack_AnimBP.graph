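# Animation Blueprint for the MediCrow backpack cosmetic, driving secondary motion on its cape.
# It copies the attached parent's pose, aims cape_top at a target, replaces spine bone transforms and caches the result.
# Booleans select between RBAN, AnimDynamics, Trail controller and DBNO cached poses, then additive rotations and bone constraints finish the output.
# AnimDynamics chains include movement-toggled variants with spherical spine limits; trails scale with backward movement.
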AnimBlueprint /BRCosmetics/Accessories/FORT_Backpacks/Backpack_M_MED_MediCrow/Meshes/M_MED_MediCrow_Pack_AnimBP
target_skeleton=&/M_MED_MediCrow_Pack_Skeleton
graphs [
 AnimGraph {nodes [
   n1 = Get bIsMoving2D()
   n2 = NOT Boolean(A=@n1)
   n3 = Anim Dynamics - Bone: dyn_cape_chain!(ComponentPose, bAlphaBoolEnabled:bool=@n2, GravityOverride:Vector="0,0,0", ExternalForce="0,0,0") -> (Pose) {
    alphaInputType=Bool
    angularDampingOverride=0.999
    angularSpringConstant=2000
    bAngularSpring=true
    bOverrideAngularDamping=true
    bUseGravityOverride=true
    bUseSphericalLimits=true
    simulationSpace=RootRelative
    alphaBoolBlend {blendInTime=2, blendOption=Linear, blendOutTime=2}
    boundBone {boneName=dyn_cape_chain}
    componentAppliedLinearAccClamp=<3000 3000 3000>
    componentLinearAccScale=<1 0.7 0.7>
    componentLinearVelScale=<1 0.7 0.7>
    physicsBodyDefinitions [{
      boundBone {boneName=dyn_cape_chain}
      boxExtents=<30 30 30>
      constraintSetup {angularLimitsMax=<5 15 3>, angularLimitsMin=<-5 -45 -3>, angularTarget=<1 0 0>}
      localJointOffset=<8 0 0>
     }]
    sphericalLimits [{limitRadius=20, drivingBone {boneName=char_spine_04}, sphereLocalOffset=<6.8575 29.1659 0.8474>}]
   }
   n4 = Anim Dynamics - Bone: dyn_cape_chain1!(ComponentPose=@n3, bAlphaBoolEnabled:bool=@n2, GravityOverride:Vector="0,0,0", ExternalForce="0,0,0") -> (Pose) {
    alphaInputType=Bool
    angularDampingOverride=0.999
    angularSpringConstant=2000
    bAngularSpring=true
    bOverrideAngularDamping=true
    bUseGravityOverride=true
    bUseSphericalLimits=true
    simulationSpace=RootRelative
    alphaBoolBlend {blendInTime=2, blendOption=Linear, blendOutTime=2}
    boundBone {boneName=dyn_cape_chain1}
    componentAppliedLinearAccClamp=<3000 3000 3000>
    componentLinearAccScale=<1 0.7 0.7>
    componentLinearVelScale=<1 0.7 0.7>
    physicsBodyDefinitions [{
      boundBone {boneName=dyn_cape_chain1}
      boxExtents=<30 30 30>
      constraintSetup {angularLimitsMax=<10 15 10>, angularLimitsMin=<-10 -45 -10>, angularTarget=<1 0 0>}
      localJointOffset=<8 0 0>
     }]
    sphericalLimits [{limitRadius=30, drivingBone {boneName=char_spine_03}, sphereLocalOffset=<0 13.4934 3.4552>}]
   }
   n5 = Anim Dynamics - Bone: dyn_cape_chain2!(ComponentPose=@n4, bAlphaBoolEnabled:bool=@n2, GravityOverride:Vector="0,0,0", ExternalForce="0,0,0") -> (Pose) {
    alphaInputType=Bool
    angularDampingOverride=0.999
    angularSpringConstant=2000
    bAngularSpring=true
    bOverrideAngularDamping=true
    bUseGravityOverride=true
    bUseSphericalLimits=true
    simulationSpace=RootRelative
    alphaBoolBlend {blendInTime=2, blendOption=Linear, blendOutTime=2}
    boundBone {boneName=dyn_cape_chain2}
    componentAppliedLinearAccClamp=<3000 3000 3000>
    componentLinearAccScale=<1 0.7 0.7>
    componentLinearVelScale=<1 0.7 0.7>
    physicsBodyDefinitions [{
      boundBone {boneName=dyn_cape_chain2}
      boxExtents=<30 30 30>
      constraintSetup {angularLimitsMax=<10 15 10>, angularLimitsMin=<-10 -45 -10>, angularTarget=<1 0 0>}
      localJointOffset=<8 0 0>
     }]
    sphericalLimits [{limitRadius=30, drivingBone {boneName=char_spine_03}, sphereLocalOffset=<0 31.5193 -4.4107>}]
   }
   n6 = Anim Dynamics - Bone: dyn_cape_chain3!(ComponentPose=@n5, bAlphaBoolEnabled:bool=@n2, GravityOverride:Vector="0,0,0", ExternalForce="0,0,0") -> (Pose) {
    alphaInputType=Bool
    angularDampingOverride=0.999
    angularSpringConstant=2000
    bAngularSpring=true
    bOverrideAngularDamping=true
    bUseGravityOverride=true
    bUseSphericalLimits=true
    simulationSpace=RootRelative
    alphaBoolBlend {blendInTime=2, blendOption=Linear, blendOutTime=2}
    boundBone {boneName=dyn_cape_chain3}
    componentAppliedLinearAccClamp=<3000 3000 3000>
    componentLinearAccScale=<1 0.7 0.7>
    componentLinearVelScale=<1 0.7 0.7>
    physicsBodyDefinitions [{
      boundBone {boneName=dyn_cape_chain3}
      boxExtents=<30 30 30>
      constraintSetup {angularLimitsMax=<10 15 10>, angularLimitsMin=<-10 -45 -10>, angularTarget=<1 0 0>}
      localJointOffset=<8 0 0>
     }]
    sphericalLimits [{limitRadius=30, drivingBone {boneName=char_spine_02}, sphereLocalOffset=<-10.7151 13.5973 -1.2787>}]
   }
   n7 = Anim Dynamics - Bone: dyn_cape_chain(ComponentPose=@n6, bAlphaBoolEnabled=@n1, GravityOverride="0,0,0", ExternalForce="0,0,0") {
    alphaInputType=Bool
    angularDampingOverride=0.999
    angularSpringConstant=2000
    bAngularSpring=true
    bOverrideAngularDamping=true
    bUseGravityOverride=true
    bUseSphericalLimits=true
    simulationSpace=RootRelative
    alphaBoolBlend {blendInTime=2, blendOption=Linear, blendOutTime=2}
    boundBone {boneName=dyn_cape_chain}
    componentAppliedLinearAccClamp=<3000 3000 3000>
    componentLinearAccScale=<1 1 0.1>
    componentLinearVelScale=<1 1 0.1>
    physicsBodyDefinitions [{
      boundBone {boneName=dyn_cape_chain}
      boxExtents=<30 30 30>
      constraintSetup {angularLimitsMax=<5 0 3>, angularLimitsMin=<-5 -70 -3>, angularTarget=<1 0 0>}
      localJointOffset=<8 0 0>
     }]
    sphericalLimits [{limitRadius=20, drivingBone {boneName=char_spine_04}, sphereLocalOffset=<6.8575 29.1659 0.8474>}]
   }
   n8 = Anim Dynamics - Bone: dyn_cape_chain1(ComponentPose=@n7, bAlphaBoolEnabled=@n1, GravityOverride="0,0,0", ExternalForce="0,0,0") {
    alphaInputType=Bool
    angularDampingOverride=0.999
    angularSpringConstant=2000
    bAngularSpring=true
    bOverrideAngularDamping=true
    bUseGravityOverride=true
    bUseSphericalLimits=true
    simulationSpace=RootRelative
    alphaBoolBlend {blendInTime=2, blendOption=Linear, blendOutTime=2}
    boundBone {boneName=dyn_cape_chain1}
    componentAppliedLinearAccClamp=<3000 3000 3000>
    componentLinearAccScale=<1 1 0.1>
    componentLinearVelScale=<1 1 0.1>
    physicsBodyDefinitions [{
      boundBone {boneName=dyn_cape_chain1}
      boxExtents=<30 30 30>
      constraintSetup {angularLimitsMax=<10 0 10>, angularLimitsMin=<-10 -45 -10>, angularTarget=<1 0 0>}
      localJointOffset=<8 0 0>
     }]
    sphericalLimits [{limitRadius=30, drivingBone {boneName=char_spine_03}, sphereLocalOffset=<0 13.4934 3.4552>}]
   }
   n9 = Anim Dynamics - Bone: dyn_cape_chain2(ComponentPose=@n8, bAlphaBoolEnabled=@n1, GravityOverride="0,0,0", ExternalForce="0,0,0") {
    alphaInputType=Bool
    angularDampingOverride=0.999
    angularSpringConstant=2000
    bAngularSpring=true
    bOverrideAngularDamping=true
    bUseGravityOverride=true
    bUseSphericalLimits=true
    simulationSpace=RootRelative
    alphaBoolBlend {blendInTime=2, blendOption=Linear, blendOutTime=2}
    boundBone {boneName=dyn_cape_chain2}
    componentAppliedLinearAccClamp=<3000 3000 3000>
    componentLinearAccScale=<1 1 0.1>
    componentLinearVelScale=<1 1 0.1>
    physicsBodyDefinitions [{
      boundBone {boneName=dyn_cape_chain2}
      boxExtents=<30 30 30>
      constraintSetup {angularLimitsMax=<10 0 10>, angularLimitsMin=<-10 -30 -10>, angularTarget=<1 0 0>}
      localJointOffset=<8 0 0>
     }]
    sphericalLimits [{limitRadius=30, drivingBone {boneName=char_spine_03}, sphereLocalOffset=<0 31.5193 -4.4107>}]
   }
   Anim Dynamics - Bone: dyn_cape_chain3(ComponentPose=@n9, bAlphaBoolEnabled=@n1, GravityOverride="0,0,0", ExternalForce="0,0,0") {
    alphaInputType=Bool
    angularDampingOverride=0.999
    angularSpringConstant=2000
    bAngularSpring=true
    bOverrideAngularDamping=true
    bUseGravityOverride=true
    bUseSphericalLimits=true
    simulationSpace=RootRelative
    alphaBoolBlend {blendInTime=2, blendOption=Linear, blendOutTime=2}
    boundBone {boneName=dyn_cape_chain3}
    componentAppliedLinearAccClamp=<3000 3000 3000>
    componentLinearAccScale=<1 1 0.1>
    componentLinearVelScale=<1 1 0.1>
    physicsBodyDefinitions [{
      boundBone {boneName=dyn_cape_chain3}
      boxExtents=<30 30 30>
      constraintSetup {angularLimitsMax=<10 0 10>, angularLimitsMin=<-10 -45 -10>, angularTarget=<1 0 0>}
      localJointOffset=<8 0 0>
     }]
    sphericalLimits [{limitRadius=30, drivingBone {boneName=char_spine_02}, sphereLocalOffset=<-10.7151 13.5973 -1.2787>}]
   }
   n10 = Anim Dynamics - Bone: dyn_cape_spine!(
    ComponentPose
    Alpha=1
    GravityOverride:Vector="0,0,0"
    ComponentLinearAccScale:Vector="0.3,0.5,0.3"
    ComponentLinearVelScale:Vector="0.3,0.5,0.3"
   ) -> (Pose) {
    angularDampingOverride=0.999
    angularSpringConstant=10000
    bAngularSpring=true
    bOverrideAngularDamping=true
    bUseGravityOverride=true
    simulationSpace=RootRelative
    alphaBoolBlend {blendOption=Linear}
    boundBone {boneName=dyn_cape_spine}
    componentAppliedLinearAccClamp=<3000 3000 3000>
    componentLinearAccScale=<0.3 0.5 0.3>
    componentLinearVelScale=<0.3 0.5 0.3>
    physicsBodyDefinitions [{
      boundBone {boneName=dyn_cape_spine}
      boxExtents=<30 30 30>
      constraintSetup {angularLimitsMax=<1.5 0 10>, angularLimitsMin=<-1.5 -20 -10>, angularTarget=<1 0 0>}
      localJointOffset=<8 0 0>
     }]
   }
   n11 = Anim Dynamics - Bone: dyn_cape_spine1!(
    ComponentPose=@n10
    Alpha=1
    GravityOverride:Vector="0,0,0"
    ComponentLinearAccScale:Vector="0.3,0.5,0.3"
    ComponentLinearVelScale:Vector="0.3,0.5,0.3"
   ) -> (Pose) {
    angularDampingOverride=0.999
    angularSpringConstant=10000
    bAngularSpring=true
    bOverrideAngularDamping=true
    bUseGravityOverride=true
    simulationSpace=RootRelative
    alphaBoolBlend {blendOption=Linear}
    boundBone {boneName=dyn_cape_spine1}
    componentAppliedLinearAccClamp=<3000 3000 3000>
    componentLinearAccScale=<0.3 0.5 0.3>
    componentLinearVelScale=<0.3 0.5 0.3>
    physicsBodyDefinitions [{
      boundBone {boneName=dyn_cape_spine1}
      boxExtents=<30 30 30>
      constraintSetup {angularLimitsMax=<1.5 0 15>, angularLimitsMin=<-1.5 -20 -15>, angularTarget=<1 0 0>}
      localJointOffset=<8 0 0>
     }]
   }
   n12 = Anim Dynamics - Bone: dyn_cape_spine2!(
    ComponentPose=@n11
    Alpha=1
    GravityOverride:Vector="0,0,0"
    ComponentLinearAccScale:Vector="0.3,0.5,0.3"
    ComponentLinearVelScale:Vector="0.3,0.5,0.3"
   ) -> (Pose) {
    angularDampingOverride=0.999
    angularSpringConstant=10000
    bAngularSpring=true
    bOverrideAngularDamping=true
    bUseGravityOverride=true
    simulationSpace=RootRelative
    alphaBoolBlend {blendOption=Linear}
    boundBone {boneName=dyn_cape_spine2}
    componentAppliedLinearAccClamp=<3000 3000 3000>
    componentLinearAccScale=<0.3 0.5 0.3>
    componentLinearVelScale=<0.3 0.5 0.3>
    physicsBodyDefinitions [{
      boundBone {boneName=dyn_cape_spine2}
      boxExtents=<30 30 30>
      constraintSetup {angularLimitsMax=<1.5 0 15>, angularLimitsMin=<-1.5 -20 -15>, angularTarget=<1 0 0>}
      localJointOffset=<8 0 0>
     }]
   }
   Anim Dynamics - Bone: dyn_cape_spine3!(
    ComponentPose=@n12
    Alpha=1
    GravityOverride:Vector="0,0,0"
    ComponentLinearAccScale:Vector="0.3,0.5,0.3"
    ComponentLinearVelScale:Vector="0.3,0.5,0.3"
   ) -> (Pose) {
    angularDampingOverride=0.999
    angularSpringConstant=10000
    bAngularSpring=true
    bOverrideAngularDamping=true
    bUseGravityOverride=true
    simulationSpace=RootRelative
    alphaBoolBlend {blendOption=Linear}
    boundBone {boneName=dyn_cape_spine3}
    componentAppliedLinearAccClamp=<3000 3000 3000>
    componentLinearAccScale=<0.3 0.5 0.3>
    componentLinearVelScale=<0.3 0.5 0.3>
    physicsBodyDefinitions [{
      boundBone {boneName=dyn_cape_spine3}
      boxExtents=<30 30 30>
      constraintSetup {angularLimitsMax=<1.5 0 15>, angularLimitsMin=<-1.5 -15 -15>, angularTarget=<1 0 0>}
      localJointOffset=<8 0 0>
     }]
   }
   n13 = Anim Dynamics - Bone: dyn_cape_01(ComponentPose=Local To Component(), Alpha=1, ExternalForce="0,0,0") {
    angularDampingOverride=1
    angularSpringConstant=10000
    bAngularSpring=true
    bOverrideAngularDamping=true
    bUseGravityOverride=true
    gravityScale=0.35
    simulationSpace=BoneRelative
    alphaBoolBlend {blendOption=Linear}
    boundBone {boneName=dyn_cape_01}
    componentAppliedLinearAccClamp=<3000 3000 3000>
    componentLinearAccScale=<2 2 2>
    componentLinearVelScale=<2 2 2>
    gravityOverride {z=-980}
    physicsBodyDefinitions [{
      boundBone {boneName=dyn_cape_01}
      boxExtents=<10 40 40>
      constraintSetup {angularLimitsMax=<0 5 15>, angularLimitsMin=<0 -40 -15>, angularTarget=<1 0 0>}
      localJointOffset=<5 0 0>
     }]
    relativeSpaceBone {boneName=backpack_root}
   }
   n14 = Anim Dynamics - Bone: dyn_cape_02(ComponentPose=@n13, Alpha=1, ExternalForce="0,0,0") {
    angularDampingOverride=1
    angularSpringConstant=10000
    bAngularSpring=true
    bOverrideAngularDamping=true
    bUseGravityOverride=true
    gravityScale=0.35
    simulationSpace=BoneRelative
    alphaBoolBlend {blendOption=Linear}
    boundBone {boneName=dyn_cape_02}
    componentAppliedLinearAccClamp=<3000 3000 3000>
    componentLinearAccScale=<2 2 2>
    componentLinearVelScale=<2 2 2>
    gravityOverride {z=-980}
    physicsBodyDefinitions [{
      boundBone {boneName=dyn_cape_02}
      boxExtents=<10 40 40>
      constraintSetup {angularLimitsMax=<0 5 15>, angularLimitsMin=<0 -40 -15>, angularTarget=<1 0 0>}
      localJointOffset=<5 0 0>
     }]
    relativeSpaceBone {boneName=backpack_root}
   }
   n15 = Anim Dynamics - Bone: dyn_cape_03(ComponentPose=@n14, Alpha=1, ExternalForce="0,0,0") {
    angularDampingOverride=1
    angularSpringConstant=10000
    bAngularSpring=true
    bOverrideAngularDamping=true
    bUseGravityOverride=true
    gravityScale=0.35
    simulationSpace=BoneRelative
    alphaBoolBlend {blendOption=Linear}
    boundBone {boneName=dyn_cape_03}
    componentAppliedLinearAccClamp=<3000 3000 3000>
    componentLinearAccScale=<2 2 2>
    componentLinearVelScale=<2 2 2>
    gravityOverride {z=-980}
    physicsBodyDefinitions [{
      boundBone {boneName=dyn_cape_03}
      boxExtents=<10 40 40>
      constraintSetup {angularLimitsMax=<0 5 15>, angularLimitsMin=<0 -40 -15>, angularTarget=<1 0 0>}
      localJointOffset=<5 0 0>
     }]
    relativeSpaceBone {boneName=backpack_root}
   }
   n16 = Anim Dynamics - Bone: dyn_cape_04(ComponentPose=@n15, Alpha=1, ExternalForce="0,0,0") {
    angularDampingOverride=1
    angularSpringConstant=10000
    bAngularSpring=true
    bOverrideAngularDamping=true
    bUseGravityOverride=true
    gravityScale=0.35
    simulationSpace=BoneRelative
    alphaBoolBlend {blendOption=Linear}
    boundBone {boneName=dyn_cape_04}
    componentAppliedLinearAccClamp=<3000 3000 3000>
    componentLinearAccScale=<2 2 2>
    componentLinearVelScale=<2 2 2>
    gravityOverride {z=-980}
    physicsBodyDefinitions [{
      boundBone {boneName=dyn_cape_04}
      boxExtents=<10 40 40>
      constraintSetup {angularLimitsMax=<0 5 15>, angularLimitsMin=<0 -40 -15>, angularTarget=<1 0 0>}
      localJointOffset=<5 0 0>
     }]
    relativeSpaceBone {boneName=backpack_root}
   }
   n17 = Anim Dynamics - Bone: dyn_cape_05(ComponentPose=@n16, Alpha=1, ExternalForce="0,0,0") {
    angularDampingOverride=1
    angularSpringConstant=10000
    bAngularSpring=true
    bOverrideAngularDamping=true
    bUseGravityOverride=true
    gravityScale=0.35
    simulationSpace=BoneRelative
    alphaBoolBlend {blendOption=Linear}
    boundBone {boneName=dyn_cape_05}
    componentAppliedLinearAccClamp=<3000 3000 3000>
    componentLinearAccScale=<2 2 2>
    componentLinearVelScale=<2 2 2>
    gravityOverride {z=-980}
    physicsBodyDefinitions [{
      boundBone {boneName=dyn_cape_05}
      boxExtents=<10 40 40>
      constraintSetup {angularLimitsMax=<0 5 15>, angularLimitsMin=<0 -40 -15>, angularTarget=<1 0 0>}
      localJointOffset=<5 0 0>
     }]
    relativeSpaceBone {boneName=backpack_root}
   }
   n18 = Anim Dynamics - Bone: dyn_cape_06(ComponentPose=@n17, Alpha=1, ExternalForce="0,0,0") {
    angularDampingOverride=1
    angularSpringConstant=10000
    bAngularSpring=true
    bOverrideAngularDamping=true
    bUseGravityOverride=true
    gravityScale=0.35
    simulationSpace=BoneRelative
    alphaBoolBlend {blendOption=Linear}
    boundBone {boneName=dyn_cape_06}
    componentAppliedLinearAccClamp=<3000 3000 3000>
    componentLinearAccScale=<2 2 2>
    componentLinearVelScale=<2 2 2>
    gravityOverride {z=-980}
    physicsBodyDefinitions [{
      boundBone {boneName=dyn_cape_06}
      boxExtents=<10 40 40>
      constraintSetup {angularLimitsMax=<0 5 15>, angularLimitsMin=<0 -40 -15>, angularTarget=<1 0 0>}
      localJointOffset=<5 0 0>
     }]
    relativeSpaceBone {boneName=backpack_root}
   }
   n19 = Anim Dynamics - Bone: dyn_cape_07(ComponentPose=@n18, Alpha=1, ExternalForce="0,0,0") {
    angularDampingOverride=1
    angularSpringConstant=10000
    bAngularSpring=true
    bOverrideAngularDamping=true
    bUseGravityOverride=true
    gravityScale=0.35
    simulationSpace=BoneRelative
    alphaBoolBlend {blendOption=Linear}
    boundBone {boneName=dyn_cape_07}
    componentAppliedLinearAccClamp=<3000 3000 3000>
    componentLinearAccScale=<2 2 2>
    componentLinearVelScale=<2 2 2>
    gravityOverride {z=-980}
    physicsBodyDefinitions [{
      boundBone {boneName=dyn_cape_07}
      boxExtents=<10 40 40>
      constraintSetup {angularLimitsMax=<0 5 15>, angularLimitsMin=<0 -40 -15>, angularTarget=<1 0 0>}
      localJointOffset=<5 0 0>
     }]
    relativeSpaceBone {boneName=backpack_root}
   }
   n20 = Anim Dynamics - Bone: dyn_cape_r_01(ComponentPose=@n19, Alpha=1, ExternalForce="0,0,0") {
    angularDampingOverride=1
    angularSpringConstant=10000
    bAngularSpring=true
    bOverrideAngularDamping=true
    bUseGravityOverride=true
    gravityScale=0.35
    simulationSpace=BoneRelative
    alphaBoolBlend {blendOption=Linear}
    boundBone {boneName=dyn_cape_r_01}
    componentAppliedLinearAccClamp=<3000 3000 3000>
    componentLinearAccScale=<2 2 2>
    componentLinearVelScale=<2 2 2>
    gravityOverride {z=-980}
    physicsBodyDefinitions [{
      boundBone {boneName=dyn_cape_r_01}
      boxExtents=<10 40 40>
      constraintSetup {angularLimitsMax=<0 5 15>, angularLimitsMin=<0 -40 -15>, angularTarget=<1 0 0>}
      localJointOffset=<5 0 0>
     }]
    relativeSpaceBone {boneName=backpack_root}
   }
   n21 = Anim Dynamics - Bone: dyn_cape_r_02(ComponentPose=@n20, Alpha=1, ExternalForce="0,0,0") {
    angularDampingOverride=1
    angularSpringConstant=10000
    bAngularSpring=true
    bOverrideAngularDamping=true
    bUseGravityOverride=true
    gravityScale=0.35
    simulationSpace=BoneRelative
    alphaBoolBlend {blendOption=Linear}
    boundBone {boneName=dyn_cape_r_02}
    componentAppliedLinearAccClamp=<3000 3000 3000>
    componentLinearAccScale=<2 2 2>
    componentLinearVelScale=<2 2 2>
    gravityOverride {z=-980}
    physicsBodyDefinitions [{
      boundBone {boneName=dyn_cape_r_02}
      boxExtents=<10 40 40>
      constraintSetup {angularLimitsMax=<0 5 15>, angularLimitsMin=<0 -40 -15>, angularTarget=<1 0 0>}
      localJointOffset=<5 0 0>
     }]
    relativeSpaceBone {boneName=backpack_root}
   }
   n22 = Anim Dynamics - Bone: dyn_cape_r_03(ComponentPose=@n21, Alpha=1, ExternalForce="0,0,0") {
    angularDampingOverride=1
    angularSpringConstant=10000
    bAngularSpring=true
    bOverrideAngularDamping=true
    bUseGravityOverride=true
    gravityScale=0.35
    simulationSpace=BoneRelative
    alphaBoolBlend {blendOption=Linear}
    boundBone {boneName=dyn_cape_r_03}
    componentAppliedLinearAccClamp=<3000 3000 3000>
    componentLinearAccScale=<2 2 2>
    componentLinearVelScale=<2 2 2>
    gravityOverride {z=-980}
    physicsBodyDefinitions [{
      boundBone {boneName=dyn_cape_r_03}
      boxExtents=<10 40 40>
      constraintSetup {angularLimitsMax=<0 5 15>, angularLimitsMin=<0 -40 -15>, angularTarget=<1 0 0>}
      localJointOffset=<5 0 0>
     }]
    relativeSpaceBone {boneName=backpack_root}
   }
   n23 = Anim Dynamics - Bone: dyn_cape_r_04(ComponentPose=@n22, Alpha=1, ExternalForce="0,0,0") {
    angularDampingOverride=1
    angularSpringConstant=10000
    bAngularSpring=true
    bOverrideAngularDamping=true
    bUseGravityOverride=true
    gravityScale=0.35
    simulationSpace=BoneRelative
    alphaBoolBlend {blendOption=Linear}
    boundBone {boneName=dyn_cape_r_04}
    componentAppliedLinearAccClamp=<3000 3000 3000>
    componentLinearAccScale=<2 2 2>
    componentLinearVelScale=<2 2 2>
    gravityOverride {z=-980}
    physicsBodyDefinitions [{
      boundBone {boneName=dyn_cape_r_04}
      boxExtents=<10 40 40>
      constraintSetup {angularLimitsMax=<0 5 15>, angularLimitsMin=<0 -40 -15>, angularTarget=<1 0 0>}
      localJointOffset=<5 0 0>
     }]
    relativeSpaceBone {boneName=backpack_root}
   }
   n24 = Anim Dynamics - Bone: dyn_cape_r_05(ComponentPose=@n23, Alpha=1, ExternalForce="0,0,0") {
    angularDampingOverride=1
    angularSpringConstant=10000
    bAngularSpring=true
    bOverrideAngularDamping=true
    bUseGravityOverride=true
    gravityScale=0.35
    simulationSpace=BoneRelative
    alphaBoolBlend {blendOption=Linear}
    boundBone {boneName=dyn_cape_r_05}
    componentAppliedLinearAccClamp=<3000 3000 3000>
    componentLinearAccScale=<2 2 2>
    componentLinearVelScale=<2 2 2>
    gravityOverride {z=-980}
    physicsBodyDefinitions [{
      boundBone {boneName=dyn_cape_r_05}
      boxExtents=<10 40 40>
      constraintSetup {angularLimitsMax=<0 5 15>, angularLimitsMin=<0 -40 -15>, angularTarget=<1 0 0>}
      localJointOffset=<5 0 0>
     }]
    relativeSpaceBone {boneName=backpack_root}
   }
   n25 = Anim Dynamics - Bone: dyn_cape_r_06(ComponentPose=@n24, Alpha=1, ExternalForce="0,0,0") {
    angularDampingOverride=1
    angularSpringConstant=10000
    bAngularSpring=true
    bOverrideAngularDamping=true
    bUseGravityOverride=true
    gravityScale=0.35
    simulationSpace=BoneRelative
    alphaBoolBlend {blendOption=Linear}
    boundBone {boneName=dyn_cape_r_06}
    componentAppliedLinearAccClamp=<3000 3000 3000>
    componentLinearAccScale=<2 2 2>
    componentLinearVelScale=<2 2 2>
    gravityOverride {z=-980}
    physicsBodyDefinitions [{
      boundBone {boneName=dyn_cape_r_06}
      boxExtents=<10 40 40>
      constraintSetup {angularLimitsMax=<0 5 15>, angularLimitsMin=<0 -40 -15>, angularTarget=<1 0 0>}
      localJointOffset=<5 0 0>
     }]
    relativeSpaceBone {boneName=backpack_root}
   }
   n26 = Anim Dynamics - Bone: dyn_cape_r_07(ComponentPose=@n25, Alpha=1, ExternalForce="0,0,0") {
    angularDampingOverride=1
    angularSpringConstant=10000
    bAngularSpring=true
    bOverrideAngularDamping=true
    bUseGravityOverride=true
    gravityScale=0.35
    simulationSpace=BoneRelative
    alphaBoolBlend {blendOption=Linear}
    boundBone {boneName=dyn_cape_r_07}
    componentAppliedLinearAccClamp=<3000 3000 3000>
    componentLinearAccScale=<2 2 2>
    componentLinearVelScale=<2 2 2>
    gravityOverride {z=-980}
    physicsBodyDefinitions [{
      boundBone {boneName=dyn_cape_r_07}
      boxExtents=<10 40 40>
      constraintSetup {angularLimitsMax=<0 5 15>, angularLimitsMin=<0 -40 -15>, angularTarget=<1 0 0>}
      localJointOffset=<5 0 0>
     }]
    relativeSpaceBone {boneName=backpack_root}
   }
   n27 = Anim Dynamics - Bone: dyn_cape_l_01(ComponentPose=@n26, Alpha=1, ExternalForce="0,0,0") {
    angularDampingOverride=1
    angularSpringConstant=10000
    bAngularSpring=true
    bOverrideAngularDamping=true
    bUseGravityOverride=true
    gravityScale=0.35
    simulationSpace=BoneRelative
    alphaBoolBlend {blendOption=Linear}
    boundBone {boneName=dyn_cape_l_01}
    componentAppliedLinearAccClamp=<3000 3000 3000>
    componentLinearAccScale=<2 2 2>
    componentLinearVelScale=<2 2 2>
    gravityOverride {z=-980}
    physicsBodyDefinitions [{
      boundBone {boneName=dyn_cape_l_01}
      boxExtents=<10 40 40>
      constraintSetup {angularLimitsMax=<0 5 15>, angularLimitsMin=<0 -40 -15>, angularTarget=<1 0 0>}
      localJointOffset=<5 0 0>
     }]
    relativeSpaceBone {boneName=backpack_root}
   }
   n28 = Anim Dynamics - Bone: dyn_cape_l_02(ComponentPose=@n27, Alpha=1, ExternalForce="0,0,0") {
    angularDampingOverride=1
    angularSpringConstant=10000
    bAngularSpring=true
    bOverrideAngularDamping=true
    bUseGravityOverride=true
    gravityScale=0.35
    simulationSpace=BoneRelative
    alphaBoolBlend {blendOption=Linear}
    boundBone {boneName=dyn_cape_l_02}
    componentAppliedLinearAccClamp=<3000 3000 3000>
    componentLinearAccScale=<2 2 2>
    componentLinearVelScale=<2 2 2>
    gravityOverride {z=-980}
    physicsBodyDefinitions [{
      boundBone {boneName=dyn_cape_l_02}
      boxExtents=<10 40 40>
      constraintSetup {angularLimitsMax=<0 5 15>, angularLimitsMin=<0 -40 -15>, angularTarget=<1 0 0>}
      localJointOffset=<5 0 0>
     }]
    relativeSpaceBone {boneName=backpack_root}
   }
   n29 = Anim Dynamics - Bone: dyn_cape_l_03(ComponentPose=@n28, Alpha=1, ExternalForce="0,0,0") {
    angularDampingOverride=1
    angularSpringConstant=10000
    bAngularSpring=true
    bOverrideAngularDamping=true
    bUseGravityOverride=true
    gravityScale=0.35
    simulationSpace=BoneRelative
    alphaBoolBlend {blendOption=Linear}
    boundBone {boneName=dyn_cape_l_03}
    componentAppliedLinearAccClamp=<3000 3000 3000>
    componentLinearAccScale=<2 2 2>
    componentLinearVelScale=<2 2 2>
    gravityOverride {z=-980}
    physicsBodyDefinitions [{
      boundBone {boneName=dyn_cape_l_03}
      boxExtents=<10 40 40>
      constraintSetup {angularLimitsMax=<0 5 15>, angularLimitsMin=<0 -40 -15>, angularTarget=<1 0 0>}
      localJointOffset=<5 0 0>
     }]
    relativeSpaceBone {boneName=backpack_root}
   }
   n30 = Anim Dynamics - Bone: dyn_cape_l_04(ComponentPose=@n29, Alpha=1, ExternalForce="0,0,0") {
    angularDampingOverride=1
    angularSpringConstant=10000
    bAngularSpring=true
    bOverrideAngularDamping=true
    bUseGravityOverride=true
    gravityScale=0.35
    simulationSpace=BoneRelative
    alphaBoolBlend {blendOption=Linear}
    boundBone {boneName=dyn_cape_l_04}
    componentAppliedLinearAccClamp=<3000 3000 3000>
    componentLinearAccScale=<2 2 2>
    componentLinearVelScale=<2 2 2>
    gravityOverride {z=-980}
    physicsBodyDefinitions [{
      boundBone {boneName=dyn_cape_l_04}
      boxExtents=<10 40 40>
      constraintSetup {angularLimitsMax=<0 5 15>, angularLimitsMin=<0 -40 -15>, angularTarget=<1 0 0>}
      localJointOffset=<5 0 0>
     }]
    relativeSpaceBone {boneName=backpack_root}
   }
   n31 = Anim Dynamics - Bone: dyn_cape_l_05(ComponentPose=@n30, Alpha=1, ExternalForce="0,0,0") {
    angularDampingOverride=1
    angularSpringConstant=10000
    bAngularSpring=true
    bOverrideAngularDamping=true
    bUseGravityOverride=true
    gravityScale=0.35
    simulationSpace=BoneRelative
    alphaBoolBlend {blendOption=Linear}
    boundBone {boneName=dyn_cape_l_05}
    componentAppliedLinearAccClamp=<3000 3000 3000>
    componentLinearAccScale=<2 2 2>
    componentLinearVelScale=<2 2 2>
    gravityOverride {z=-980}
    physicsBodyDefinitions [{
      boundBone {boneName=dyn_cape_l_05}
      boxExtents=<10 40 40>
      constraintSetup {angularLimitsMax=<0 5 15>, angularLimitsMin=<0 -40 -15>, angularTarget=<1 0 0>}
      localJointOffset=<5 0 0>
     }]
    relativeSpaceBone {boneName=backpack_root}
   }
   n32 = Anim Dynamics - Bone: dyn_cape_l_06(ComponentPose=@n31, Alpha=1, ExternalForce="0,0,0") {
    angularDampingOverride=1
    angularSpringConstant=10000
    bAngularSpring=true
    bOverrideAngularDamping=true
    bUseGravityOverride=true
    gravityScale=0.35
    simulationSpace=BoneRelative
    alphaBoolBlend {blendOption=Linear}
    boundBone {boneName=dyn_cape_l_06}
    componentAppliedLinearAccClamp=<3000 3000 3000>
    componentLinearAccScale=<2 2 2>
    componentLinearVelScale=<2 2 2>
    gravityOverride {z=-980}
    physicsBodyDefinitions [{
      boundBone {boneName=dyn_cape_l_06}
      boxExtents=<10 40 40>
      constraintSetup {angularLimitsMax=<0 5 15>, angularLimitsMin=<0 -40 -15>, angularTarget=<1 0 0>}
      localJointOffset=<5 0 0>
     }]
    relativeSpaceBone {boneName=backpack_root}
   }
   n33 = Anim Dynamics - Bone: dyn_cape_l_07(ComponentPose=@n32, Alpha=1, ExternalForce="0,0,0") {
    angularDampingOverride=1
    angularSpringConstant=10000
    bAngularSpring=true
    bOverrideAngularDamping=true
    bUseGravityOverride=true
    gravityScale=0.35
    simulationSpace=BoneRelative
    alphaBoolBlend {blendOption=Linear}
    boundBone {boneName=dyn_cape_l_07}
    componentAppliedLinearAccClamp=<3000 3000 3000>
    componentLinearAccScale=<2 2 2>
    componentLinearVelScale=<2 2 2>
    gravityOverride {z=-980}
    physicsBodyDefinitions [{
      boundBone {boneName=dyn_cape_l_07}
      boxExtents=<10 40 40>
      constraintSetup {angularLimitsMax=<0 5 15>, angularLimitsMin=<0 -40 -15>, angularTarget=<1 0 0>}
      localJointOffset=<5 0 0>
     }]
    relativeSpaceBone {boneName=backpack_root}
   }
   Component To Local(ComponentPose=@n33)
   n34 = Blend Poses by bool(
    BlendPose_0=Use cached pose 'animDyn Cache'()
    BlendPose_1=Blend Poses by bool(BlendPose_0=Use cached pose 'dbnoCache'(), BlendPose_1=Use cached pose 'trailCache'(), bActiveValue=Get bIsDBNO()) {blendType=Linear}
    bActiveValue=Get bAnimDynamicsEnabled() -> (bAnimDynamicsEnabled:bool)
   ) {blendType=Linear}
   n35 = Use cached pose 'RbanCache'()
   n36 = Get RbanOrCloth() -> (RbanOrCloth:bool)
   n37 = Get RotationBP() -> (RotationBP:Rotator)
   n38 = "Transform (Modify) Bone - Bone: dyn_l_newchain_joint1"!(
    ComponentPose=Local To Component(LocalPose=Blend Poses by bool(BlendPose_0=@n35, BlendPose_1=@n34, bActiveValue=@n36) {blendType=Linear})
    bAlphaBoolEnabled:bool=True
    Translation="0,0,0"
    Rotation=@n37
    Scale="1,1,1"
   ) -> (Pose) {
    alphaInputType=Bool
    rotationMode=BMM_Additive
    rotationSpace=BCS_BoneSpace
    translationMode=BMM_Additive
    alphaBoolBlend {blendInTime=0.2, blendOption=Linear, blendOutTime=0.2}
    boneToModify {boneName=dyn_l_newchain_joint1}
   }
   n39 = "Transform (Modify) Bone - Bone: dyn_r_newchain_joint1"!(ComponentPose=@n38, bAlphaBoolEnabled:bool=True, Translation="0,0,0", Rotation=@n37, Scale="1,1,1") -> (Pose) {
    alphaInputType=Bool
    rotationMode=BMM_Additive
    rotationSpace=BCS_BoneSpace
    translationMode=BMM_Additive
    alphaBoolBlend {blendInTime=0.2, blendOption=Linear, blendOutTime=0.2}
    boneToModify {boneName=dyn_r_newchain_joint1}
   }
   n40 = "Constraint - dyn_cape_spine1 (PP)"!(ComponentPose=@n39, Alpha=0.5, ConstraintWeights_0=0.5, ConstraintWeights_1:float=0.5) -> (Pose) {
    alphaBoolBlend {blendOption=Linear}
    boneToModify {boneName=dyn_cape_spine1}
    constraintSetup [{targetBone {boneName=dyn_l_newchain_joint1}}, {targetBone {boneName=dyn_r_newchain_joint1}}]
    constraintWeights [1,1]
   }
   n41 = "Constraint - dyn_cape_spine2 (PP)"!(ComponentPose=@n40, Alpha=0.5, ConstraintWeights_0=0.5, ConstraintWeights_1:float=0.5) -> (Pose) {
    alphaBoolBlend {blendOption=Linear}
    boneToModify {boneName=dyn_cape_spine2}
    constraintSetup [{targetBone {boneName=dyn_l_newchain_joint1}}, {targetBone {boneName=dyn_r_newchain_joint1}}]
    constraintWeights [1,1]
   }
   n42 = "Constraint - dyn_cape_spine3 (PP)"!(ComponentPose=@n41, Alpha=0.5, ConstraintWeights_0=0.5, ConstraintWeights_1:float=0.5) -> (Pose) {
    alphaBoolBlend {blendOption=Linear}
    boneToModify {boneName=dyn_cape_spine3}
    constraintSetup [{targetBone {boneName=dyn_l_newchain_joint1}}, {targetBone {boneName=dyn_r_newchain_joint1}}]
    constraintWeights [1,1]
   }
   Output Pose(Result=Component To Local(ComponentPose=@n42)) {layerGroup=None}
   n43 = Copy Pose From Mesh() {bUseAttachedParent=true}
   n44 = Get!() -> (spin02_Location:Vector, spin02_Rotation:Rotator, spin02_Scale:Vector)
   n45 = "Transform (Modify) Bone - Bone: char_spine_02"(
    ComponentPose=Look At - Bone: cape_top!(ComponentPose=Local To Component(LocalPose=@n43), Alpha=1, LookAtLocation:Vector=Get Cape_LookAt_Target() -> (Cape_LookAt_Target:Vector)) -> (Pose) {alphaBoolBlend {blendOption=Linear}, boneToModify {boneName=cape_top}, lookAt_Axis {axis {y=0, z=-1}}, lookAtTarget=~}
    Alpha=1
    Translation=@n44
    Rotation=@n44.spin02_Rotation
    Scale=@n44.spin02_Scale
   ) {
    rotationMode=BMM_Replace
    rotationSpace=BCS_WorldSpace
    scaleMode=BMM_Replace
    scaleSpace=BCS_WorldSpace
    translationMode=BMM_Replace
    translationSpace=BCS_WorldSpace
    alphaBoolBlend {blendOption=Linear}
    boneToModify {boneName=char_spine_02}
   }
   n46 = Get!() -> (spin03_Location:Vector, spin03_Rotation:Rotator, spin03_Scale:Vector)
   n47 = "Transform (Modify) Bone - Bone: char_spine_03"(ComponentPose=@n45, Alpha=1, Translation=@n46, Rotation=@n46.spin03_Rotation, Scale=@n46.spin03_Scale) {
    rotationMode=BMM_Replace
    rotationSpace=BCS_WorldSpace
    scaleMode=BMM_Replace
    scaleSpace=BCS_WorldSpace
    translationMode=BMM_Replace
    translationSpace=BCS_WorldSpace
    alphaBoolBlend {blendOption=Linear}
    boneToModify {boneName=char_spine_03}
   }
   n48 = Get!() -> (spin04_Location:Vector, spin04_Rotation:Rotator, spin04_Scale:Vector)
   n49 = "Transform (Modify) Bone - Bone: char_spine_04"(ComponentPose=@n47, Alpha=1, Translation=@n48, Rotation=@n48.spin04_Rotation, Scale=@n48.spin04_Scale) {
    rotationMode=BMM_Replace
    rotationSpace=BCS_WorldSpace
    scaleMode=BMM_Replace
    scaleSpace=BCS_WorldSpace
    translationMode=BMM_Replace
    translationSpace=BCS_WorldSpace
    alphaBoolBlend {blendOption=Linear}
    boneToModify {boneName=char_spine_04}
   }
   Save cached pose 'Start_Cache'(Pose=Component To Local(ComponentPose=@n49)) {cacheName=Start_Cache}
   n50 = Get SkydivingAlpha() -> (SkydivingAlpha:Rotator)
   n51 = "Transform (Modify) Bone - Bone: dyn_cape_l_01"(ComponentPose=Local To Component(LocalPose=Use cached pose 'Start_Cache'()), Alpha=0.5, Translation="0,0,0", Rotation=@n50, Scale="1,1,1") {rotationMode=BMM_Additive, rotationSpace=BCS_BoneSpace, alphaBoolBlend {blendOption=Linear}, boneToModify {boneName=dyn_cape_l_01}}
   n52 = "Transform (Modify) Bone - Bone: dyn_cape_01"(ComponentPose=@n51, Alpha=0.5, Translation="0,0,0", Rotation=@n50, Scale="1,1,1") {rotationMode=BMM_Additive, rotationSpace=BCS_BoneSpace, alphaBoolBlend {blendOption=Linear}, boneToModify {boneName=dyn_cape_01}}
   n53 = "Transform (Modify) Bone - Bone: dyn_cape_r_01"(ComponentPose=@n52, Alpha=0.5, Translation="0,0,0", Rotation=@n50, Scale="1,1,1") {rotationMode=BMM_Additive, rotationSpace=BCS_BoneSpace, alphaBoolBlend {blendOption=Linear}, boneToModify {boneName=dyn_cape_r_01}}
   n54 = Get BwdMovement() -> (BwdMovement:double)
   n55 = Trail controller - Bone: dyn_cape_l_07(ComponentPose=@n53, Alpha=@n54) {
    bUsePlanarLimit=true
    chainLength=6
    debugLifeTime=0.1
    maxDeltaTime=0.0333
    alphaScaleBiasClamp {bInterpResult=true}
    baseJoint {boneName=backpack_root}
    planarLimits [{drivingBone {boneName=backpack_root}, planeTransform {location=<0 0 0>, rotation=<rot 0 0 -90>, scale=<1 1 1>}}]
    rotationLimits [
     {limitMax=<180 180 180>}
     {limitMax=<180 180 180>}
     {limitMax=<180 180 180>}
     {limitMax=<180 180 180>}
     {limitMax=<180 180 180>}
     {limitMax=<180 180 180>}
    ]
    rotationOffsets [<0 0 0>, <0 0 0>, <0 0 0>, <0 0 0>, <0 0 0>, <0 0 0>]
    trailBone {boneName=dyn_cape_l_07}
    trailRelaxationSpeed {editorCurveData {keys [<at 0 23 RCIM_Linear>,<at 0.284 19.3799 RCIM_Linear>,<at 0.6554 8.5924 RCIM_Linear>,<at 1 5 RCIM_Linear>]}}
   }
   n56 = Trail controller - Bone: dyn_cape_07(ComponentPose=@n55, Alpha=@n54) {
    bUsePlanarLimit=true
    chainLength=6
    debugLifeTime=0.1
    maxDeltaTime=0.0333
    alphaScaleBiasClamp {bInterpResult=true}
    baseJoint {boneName=backpack_root}
    planarLimits [{drivingBone {boneName=backpack_root}, planeTransform {location=<0 0 0>, rotation=<rot 0 0 -90>, scale=<1 1 1>}}]
    rotationLimits [
     {limitMax=<180 180 180>}
     {limitMax=<180 180 180>}
     {limitMax=<180 90 180>, limitMin=<-180 -90 -180>}
     {limitMax=<180 180 180>}
     {limitMax=<180 180 180>}
     {limitMax=<180 180 180>}
    ]
    rotationOffsets [<0 0 0>, <0 0 0>, <0 0 0>, <0 0 0>, <0 0 0>, <0 0 0>]
    trailBone {boneName=dyn_cape_07}
    trailRelaxationSpeed {editorCurveData {keys [<at 0 23 RCIM_Linear>,<at 0.284 19.3799 RCIM_Linear>,<at 0.6554 8.5924 RCIM_Linear>,<at 1 5 RCIM_Linear>]}}
   }
   n57 = Trail controller - Bone: dyn_cape_r_07(ComponentPose=@n56, Alpha=@n54) {
    bUsePlanarLimit=true
    chainLength=6
    debugLifeTime=0.1
    maxDeltaTime=0.0333
    alphaScaleBiasClamp {bInterpResult=true}
    baseJoint {boneName=backpack_root}
    planarLimits [{drivingBone {boneName=backpack_root}, planeTransform {location=<0 0 0>, rotation=<rot 0 0 -90>, scale=<1 1 1>}}]
    rotationLimits [
     {limitMax=<180 180 180>}
     {limitMax=<180 180 180>}
     {limitMax=<180 90 180>, limitMin=<-180 -90 -180>}
     {limitMax=<180 180 180>}
     {limitMax=<180 180 180>}
     {limitMax=<180 180 180>}
    ]
    rotationOffsets [<0 0 0>, <0 0 0>, <0 0 0>, <0 0 0>, <0 0 0>, <0 0 0>]
    trailBone {boneName=dyn_cape_r_07}
    trailRelaxationSpeed {editorCurveData {keys [<at 0 23 RCIM_Linear>,<at 0.284 19.3799 RCIM_Linear>,<at 0.6554 8.5924 RCIM_Linear>,<at 1 5 RCIM_Linear>]}}
   }
   n58 = "Constraint - dyn_l_newchain_joint1 (RR)"!(ComponentPose=@n57, Alpha=0.7, ConstraintWeights_0=0.5, ConstraintWeights_1:float=0.5) -> (Pose) {
    alphaBoolBlend {blendOption=Linear}
    boneToModify {boneName=dyn_l_newchain_joint1}
    constraintSetup [{transformType=Rotation, targetBone {boneName=dyn_cape_02}}, {transformType=Rotation, targetBone {boneName=dyn_cape_03}}]
    constraintWeights [1,1]
   }
   n59 = "Constraint - dyn_r_newchain_joint1 (RR)"!(ComponentPose=@n58, Alpha=0.7, ConstraintWeights_0=0.5, ConstraintWeights_1:float=0.5) -> (Pose) {
    alphaBoolBlend {blendOption=Linear}
    boneToModify {boneName=dyn_r_newchain_joint1}
    constraintSetup [{transformType=Rotation, targetBone {boneName=dyn_cape_02}}, {transformType=Rotation, targetBone {boneName=dyn_cape_03}}]
    constraintWeights [1,1]
   }
   Save cached pose 'trailCache'(Pose=Component To Local(ComponentPose=@n59)) {cacheName=trailCache}
   n60 = "Transform (Modify) Bone - Bone: dyn_cape_l_01"(ComponentPose=Local To Component(LocalPose=Use cached pose 'Start_Cache'()), Alpha=1, Translation="0,0,0", Rotation="-10,0,0", Scale="1,1,1") {
    rotationMode=BMM_Additive
    rotationSpace=BCS_BoneSpace
    alphaBoolBlend {blendOption=Linear}
    boneToModify {boneName=dyn_cape_l_01}
    rotation {pitch=-10}
   }
   n61 = "Transform (Modify) Bone - Bone: dyn_cape_01"(ComponentPose=@n60, Alpha=1, Translation="0,0,0", Rotation="-7,0,0", Scale="1,1,1") {
    rotationMode=BMM_Additive
    rotationSpace=BCS_BoneSpace
    alphaBoolBlend {blendOption=Linear}
    boneToModify {boneName=dyn_cape_01}
    rotation {pitch=-7}
   }
   n62 = "Transform (Modify) Bone - Bone: dyn_cape_r_01"(ComponentPose=@n61, Alpha=1, Translation="0,0,0", Rotation="0,0,0", Scale="1,1,1") {rotationMode=BMM_Additive, rotationSpace=BCS_BoneSpace, alphaBoolBlend {blendOption=Linear}, boneToModify {boneName=dyn_cape_r_01}}
   n63 = "Transform (Modify) Bone - Bone: dyn_cape_l_03"(ComponentPose=@n62, Alpha=1, Translation="0,0,0", Rotation="-10,0,0", Scale="1,1,1") {
    rotationMode=BMM_Additive
    rotationSpace=BCS_BoneSpace
    alphaBoolBlend {blendOption=Linear}
    boneToModify {boneName=dyn_cape_l_03}
    rotation {pitch=-10}
   }
   n64 = "Transform (Modify) Bone - Bone: dyn_cape_03"(ComponentPose=@n63, Alpha=1, Translation="0,0,0", Rotation="-7,0,0", Scale="1,1,1") {
    rotationMode=BMM_Additive
    rotationSpace=BCS_BoneSpace
    alphaBoolBlend {blendOption=Linear}
    boneToModify {boneName=dyn_cape_03}
    rotation {pitch=-7}
   }
   n65 = "Transform (Modify) Bone - Bone: dyn_cape_r_03"(ComponentPose=@n64, Alpha=1, Translation="0,0,0", Rotation="-3,0,0", Scale="1,1,1") {
    rotationMode=BMM_Additive
    rotationSpace=BCS_BoneSpace
    alphaBoolBlend {blendOption=Linear}
    boneToModify {boneName=dyn_cape_r_03}
    rotation {pitch=-3}
   }
   n66 = "Transform (Modify) Bone - Bone: dyn_cape_l_04"(ComponentPose=@n65, Alpha=1, Translation="0,0,0", Rotation="-10,0,0", Scale="1,1,1") {
    rotationMode=BMM_Additive
    rotationSpace=BCS_BoneSpace
    alphaBoolBlend {blendOption=Linear}
    boneToModify {boneName=dyn_cape_l_04}
    rotation {pitch=-10}
   }
   n67 = "Transform (Modify) Bone - Bone: dyn_cape_04"(ComponentPose=@n66, Alpha=1, Translation="0,0,0", Rotation="-10,0,0", Scale="1,1,1") {
    rotationMode=BMM_Additive
    rotationSpace=BCS_BoneSpace
    alphaBoolBlend {blendOption=Linear}
    boneToModify {boneName=dyn_cape_04}
    rotation {pitch=-10}
   }
   n68 = "Transform (Modify) Bone - Bone: dyn_cape_r_04"(ComponentPose=@n67, Alpha=1, Translation="0,0,0", Rotation="-20,0,0", Scale="1,1,1") {
    rotationMode=BMM_Additive
    rotationSpace=BCS_BoneSpace
    alphaBoolBlend {blendOption=Linear}
    boneToModify {boneName=dyn_cape_r_04}
    rotation {pitch=-20}
   }
   n69 = Get BwdMovement()
   n70 = Trail controller - Bone: dyn_cape_l_07(ComponentPose=@n68, Alpha=@n69) {
    bActorSpaceFakeVel=true
    bShowTrailLocation=true
    bUsePlanarLimit=true
    chainLength=7
    debugLifeTime=0.1
    maxDeltaTime=0.0333
    relaxationSpeedScale=2
    alphaScaleBiasClamp {bInterpResult=true, interpSpeedDecreasing=25, interpSpeedIncreasing=25}
    baseJoint {boneName=backpack_root}
    rotationLimits [
     {limitMax=<180 180 180>}
     {limitMax=<180 180 180>}
     {limitMax=<180 180 180>}
     {limitMax=<180 180 180>}
     {limitMax=<180 180 180>}
     {limitMax=<180 180 180>}
     {limitMax=<180 180 180>}
    ]
    rotationOffsets [<0 0 0>, <0 0 0>, <0 0 0>, <0 0 0>, <0 0 0>, <0 0 0>, <0 0 0>]
    trailBone {boneName=dyn_cape_l_07}
    trailRelaxationSpeed {editorCurveData {keys [<at 0 23 RCIM_Linear>,<at 0.284 19.3799 RCIM_Linear>,<at 0.6554 8.5924 RCIM_Linear>,<at 1 5 RCIM_Linear>]}}
   }
   n71 = Trail controller - Bone: dyn_cape_07(ComponentPose=@n70, Alpha=@n69) {
    bActorSpaceFakeVel=true
    bShowTrailLocation=true
    bUsePlanarLimit=true
    chainLength=7
    debugLifeTime=0.1
    maxDeltaTime=0.0333
    relaxationSpeedScale=2
    alphaScaleBiasClamp {bInterpResult=true, interpSpeedDecreasing=25, interpSpeedIncreasing=25}
    baseJoint {boneName=backpack_root}
    rotationLimits [
     {limitMax=<180 180 180>}
     {limitMax=<180 180 180>}
     {limitMax=<180 180 180>}
     {limitMax=<180 180 180>}
     {limitMax=<180 180 180>}
     {limitMax=<180 180 180>}
     {limitMax=<180 180 180>}
    ]
    rotationOffsets [<0 0 0>, <0 0 0>, <0 0 0>, <0 0 0>, <0 0 0>, <0 0 0>, <0 0 0>]
    trailBone {boneName=dyn_cape_07}
    trailRelaxationSpeed {editorCurveData {keys [<at 0 23 RCIM_Linear>,<at 0.284 19.3799 RCIM_Linear>,<at 0.6554 8.5924 RCIM_Linear>,<at 1 5 RCIM_Linear>]}}
   }
   n72 = Trail controller - Bone: dyn_cape_r_07(ComponentPose=@n71, Alpha=@n69) {
    bActorSpaceFakeVel=true
    bShowTrailLocation=true
    bUsePlanarLimit=true
    chainLength=7
    debugLifeTime=0.1
    maxDeltaTime=0.0333
    relaxationSpeedScale=2
    alphaScaleBiasClamp {bInterpResult=true, interpSpeedDecreasing=25, interpSpeedIncreasing=25}
    baseJoint {boneName=backpack_root}
    rotationLimits [
     {limitMax=<180 180 180>}
     {limitMax=<180 180 180>}
     {limitMax=<180 180 180>}
     {limitMax=<180 180 180>}
     {limitMax=<180 180 180>}
     {limitMax=<180 180 180>}
     {limitMax=<180 180 180>}
    ]
    rotationOffsets [<0 0 0>, <0 0 0>, <0 0 0>, <0 0 0>, <0 0 0>, <0 0 0>, <0 0 0>]
    trailBone {boneName=dyn_cape_r_07}
    trailRelaxationSpeed {editorCurveData {keys [<at 0 23 RCIM_Linear>,<at 0.284 19.3799 RCIM_Linear>,<at 0.6554 8.5924 RCIM_Linear>,<at 1 5 RCIM_Linear>]}}
   }
   Save cached pose 'dbnoCache'(Pose=Component To Local(ComponentPose=@n72)) {cacheName=dbnoCache}
   n73 = Anim Dynamics - Bone: dyn_cape_01(ComponentPose=Local To Component(LocalPose=Use cached pose 'Start_Cache'()), Alpha=1, ExternalForce="0,0,0") {
    angularDampingOverride=1
    angularSpringConstant=10000
    bAngularSpring=true
    bOverrideAngularDamping=true
    bUseGravityOverride=true
    gravityScale=0.35
    simulationSpace=BoneRelative
    alphaBoolBlend {blendOption=Linear}
    boundBone {boneName=dyn_cape_01}
    componentAppliedLinearAccClamp=<3000 3000 3000>
    componentLinearAccScale=<1 1 1>
    componentLinearVelScale=<1 1 1>
    gravityOverride {z=-980}
    physicsBodyDefinitions [{
      boundBone {boneName=dyn_cape_01}
      boxExtents=<10 40 40>
      constraintSetup {angularLimitsMax=<0 5 15>, angularLimitsMin=<0 -10 -15>, angularTarget=<1 0 0>}
      localJointOffset=<5 0 0>
     }]
    relativeSpaceBone {boneName=backpack_root}
   }
   n74 = Anim Dynamics - Bone: dyn_cape_03(ComponentPose=@n73, Alpha=1, ExternalForce="0,0,0") {
    angularDampingOverride=1
    angularSpringConstant=10000
    bAngularSpring=true
    bOverrideAngularDamping=true
    bUseGravityOverride=true
    gravityScale=0.35
    simulationSpace=BoneRelative
    alphaBoolBlend {blendOption=Linear}
    boundBone {boneName=dyn_cape_03}
    componentAppliedLinearAccClamp=<3000 3000 3000>
    componentLinearAccScale=<2 2 2>
    componentLinearVelScale=<2 2 2>
    gravityOverride {z=-980}
    physicsBodyDefinitions [{
      boundBone {boneName=dyn_cape_03}
      boxExtents=<10 40 40>
      constraintSetup {angularLimitsMax=<0 5 15>, angularLimitsMin=<0 -40 -15>, angularTarget=<1 0 0>}
      localJointOffset=<5 0 0>
     }]
    relativeSpaceBone {boneName=backpack_root}
   }
   n75 = Anim Dynamics - Bone: dyn_cape_05(ComponentPose=@n74, Alpha=1, ExternalForce="0,0,0") {
    angularDampingOverride=1
    angularSpringConstant=10000
    bAngularSpring=true
    bOverrideAngularDamping=true
    bUseGravityOverride=true
    gravityScale=0.35
    simulationSpace=BoneRelative
    alphaBoolBlend {blendOption=Linear}
    boundBone {boneName=dyn_cape_05}
    componentAppliedLinearAccClamp=<3000 3000 3000>
    componentLinearAccScale=<2 2 2>
    componentLinearVelScale=<2 2 2>
    gravityOverride {z=-980}
    physicsBodyDefinitions [{
      boundBone {boneName=dyn_cape_05}
      boxExtents=<10 40 40>
      constraintSetup {angularLimitsMax=<0 5 15>, angularLimitsMin=<0 -40 -15>, angularTarget=<1 0 0>}
      localJointOffset=<5 0 0>
     }]
    relativeSpaceBone {boneName=backpack_root}
   }
   n76 = Anim Dynamics - Bone: dyn_cape_06(ComponentPose=@n75, Alpha=1, ExternalForce="0,0,0") {
    angularDampingOverride=1
    angularSpringConstant=10000
    bAngularSpring=true
    bOverrideAngularDamping=true
    bUseGravityOverride=true
    gravityScale=0.35
    simulationSpace=BoneRelative
    alphaBoolBlend {blendOption=Linear}
    boundBone {boneName=dyn_cape_06}
    componentAppliedLinearAccClamp=<3000 3000 3000>
    componentLinearAccScale=<2 2 2>
    componentLinearVelScale=<2 2 2>
    gravityOverride {z=-980}
    physicsBodyDefinitions [{
      boundBone {boneName=dyn_cape_06}
      boxExtents=<10 40 40>
      constraintSetup {angularLimitsMax=<0 5 15>, angularLimitsMin=<0 -40 -15>, angularTarget=<1 0 0>}
      localJointOffset=<5 0 0>
     }]
    relativeSpaceBone {boneName=backpack_root}
   }
   n77 = Anim Dynamics - Bone: dyn_cape_r_01(ComponentPose=@n76, Alpha=1, ExternalForce="0,0,0") {
    angularDampingOverride=1
    angularSpringConstant=10000
    bAngularSpring=true
    bOverrideAngularDamping=true
    bUseGravityOverride=true
    gravityScale=0.35
    simulationSpace=BoneRelative
    alphaBoolBlend {blendOption=Linear}
    boundBone {boneName=dyn_cape_r_01}
    componentAppliedLinearAccClamp=<3000 3000 3000>
    componentLinearAccScale=<1 1 1>
    componentLinearVelScale=<1 1 1>
    gravityOverride {z=-980}
    physicsBodyDefinitions [{
      boundBone {boneName=dyn_cape_r_01}
      boxExtents=<10 40 40>
      constraintSetup {angularLimitsMax=<0 5 15>, angularLimitsMin=<0 -10 -15>, angularTarget=<1 0 0>}
      localJointOffset=<5 0 0>
     }]
    relativeSpaceBone {boneName=backpack_root}
   }
   n78 = Anim Dynamics - Bone: dyn_cape_r_03(ComponentPose=@n77, Alpha=1, ExternalForce="0,0,0") {
    angularDampingOverride=1
    angularSpringConstant=10000
    bAngularSpring=true
    bOverrideAngularDamping=true
    bUseGravityOverride=true
    gravityScale=0.35
    simulationSpace=BoneRelative
    alphaBoolBlend {blendOption=Linear}
    boundBone {boneName=dyn_cape_r_03}
    componentAppliedLinearAccClamp=<3000 3000 3000>
    componentLinearAccScale=<2 2 2>
    componentLinearVelScale=<2 2 2>
    gravityOverride {z=-980}
    physicsBodyDefinitions [{
      boundBone {boneName=dyn_cape_r_03}
      boxExtents=<10 40 40>
      constraintSetup {angularLimitsMax=<0 5 15>, angularLimitsMin=<0 -40 -15>, angularTarget=<1 0 0>}
      localJointOffset=<5 0 0>
     }]
    relativeSpaceBone {boneName=backpack_root}
   }
   n79 = Anim Dynamics - Bone: dyn_cape_r_05(ComponentPose=@n78, Alpha=1, ExternalForce="0,0,0") {
    angularDampingOverride=1
    angularSpringConstant=10000
    bAngularSpring=true
    bOverrideAngularDamping=true
    bUseGravityOverride=true
    gravityScale=0.35
    simulationSpace=BoneRelative
    alphaBoolBlend {blendOption=Linear}
    boundBone {boneName=dyn_cape_r_05}
    componentAppliedLinearAccClamp=<3000 3000 3000>
    componentLinearAccScale=<2 2 2>
    componentLinearVelScale=<2 2 2>
    gravityOverride {z=-980}
    physicsBodyDefinitions [{
      boundBone {boneName=dyn_cape_r_05}
      boxExtents=<10 40 40>
      constraintSetup {angularLimitsMax=<0 5 15>, angularLimitsMin=<0 -40 -15>, angularTarget=<1 0 0>}
      localJointOffset=<5 0 0>
     }]
    relativeSpaceBone {boneName=backpack_root}
   }
   n80 = Anim Dynamics - Bone: dyn_cape_r_06(ComponentPose=@n79, Alpha=1, ExternalForce="0,0,0") {
    angularDampingOverride=1
    angularSpringConstant=10000
    bAngularSpring=true
    bOverrideAngularDamping=true
    bUseGravityOverride=true
    gravityScale=0.35
    simulationSpace=BoneRelative
    alphaBoolBlend {blendOption=Linear}
    boundBone {boneName=dyn_cape_r_06}
    componentAppliedLinearAccClamp=<3000 3000 3000>
    componentLinearAccScale=<2 2 2>
    componentLinearVelScale=<2 2 2>
    gravityOverride {z=-980}
    physicsBodyDefinitions [{
      boundBone {boneName=dyn_cape_r_06}
      boxExtents=<10 40 40>
      constraintSetup {angularLimitsMax=<0 5 15>, angularLimitsMin=<0 -40 -15>, angularTarget=<1 0 0>}
      localJointOffset=<5 0 0>
     }]
    relativeSpaceBone {boneName=backpack_root}
   }
   n81 = Anim Dynamics - Bone: dyn_cape_l_01(ComponentPose=@n80, Alpha=1, ExternalForce="0,0,0") {
    angularDampingOverride=1
    angularSpringConstant=10000
    bAngularSpring=true
    bOverrideAngularDamping=true
    bUseGravityOverride=true
    gravityScale=0.35
    simulationSpace=BoneRelative
    alphaBoolBlend {blendOption=Linear}
    boundBone {boneName=dyn_cape_l_01}
    componentAppliedLinearAccClamp=<3000 3000 3000>
    componentLinearAccScale=<1 1 1>
    componentLinearVelScale=<1 1 1>
    gravityOverride {z=-980}
    physicsBodyDefinitions [{
      boundBone {boneName=dyn_cape_l_01}
      boxExtents=<10 40 40>
      constraintSetup {angularLimitsMax=<0 5 15>, angularLimitsMin=<0 -10 -15>, angularTarget=<1 0 0>}
      localJointOffset=<5 0 0>
     }]
    relativeSpaceBone {boneName=backpack_root}
   }
   n82 = Anim Dynamics - Bone: dyn_cape_l_03(ComponentPose=@n81, Alpha=1, ExternalForce="0,0,0") {
    angularDampingOverride=1
    angularSpringConstant=10000
    bAngularSpring=true
    bOverrideAngularDamping=true
    bUseGravityOverride=true
    gravityScale=0.35
    simulationSpace=BoneRelative
    alphaBoolBlend {blendOption=Linear}
    boundBone {boneName=dyn_cape_l_03}
    componentAppliedLinearAccClamp=<3000 3000 3000>
    componentLinearAccScale=<2 2 2>
    componentLinearVelScale=<2 2 2>
    gravityOverride {z=-980}
    physicsBodyDefinitions [{
      boundBone {boneName=dyn_cape_l_03}
      boxExtents=<10 40 40>
      constraintSetup {angularLimitsMax=<0 5 15>, angularLimitsMin=<0 -40 -15>, angularTarget=<1 0 0>}
      localJointOffset=<5 0 0>
     }]
    relativeSpaceBone {boneName=backpack_root}
   }
   n83 = Anim Dynamics - Bone: dyn_cape_l_05(ComponentPose=@n82, Alpha=1, ExternalForce="0,0,0") {
    angularDampingOverride=1
    angularSpringConstant=10000
    bAngularSpring=true
    bOverrideAngularDamping=true
    bUseGravityOverride=true
    gravityScale=0.35
    simulationSpace=BoneRelative
    alphaBoolBlend {blendOption=Linear}
    boundBone {boneName=dyn_cape_l_05}
    componentAppliedLinearAccClamp=<3000 3000 3000>
    componentLinearAccScale=<2 2 2>
    componentLinearVelScale=<2 2 2>
    gravityOverride {z=-980}
    physicsBodyDefinitions [{
      boundBone {boneName=dyn_cape_l_05}
      boxExtents=<10 40 40>
      constraintSetup {angularLimitsMax=<0 5 15>, angularLimitsMin=<0 -40 -15>, angularTarget=<1 0 0>}
      localJointOffset=<5 0 0>
     }]
    relativeSpaceBone {boneName=backpack_root}
   }
   n84 = Anim Dynamics - Bone: dyn_cape_l_06(ComponentPose=@n83, Alpha=1, ExternalForce="0,0,0") {
    angularDampingOverride=1
    angularSpringConstant=10000
    bAngularSpring=true
    bOverrideAngularDamping=true
    bUseGravityOverride=true
    gravityScale=0.35
    simulationSpace=BoneRelative
    alphaBoolBlend {blendOption=Linear}
    boundBone {boneName=dyn_cape_l_06}
    componentAppliedLinearAccClamp=<3000 3000 3000>
    componentLinearAccScale=<2 2 2>
    componentLinearVelScale=<2 2 2>
    gravityOverride {z=-980}
    physicsBodyDefinitions [{
      boundBone {boneName=dyn_cape_l_06}
      boxExtents=<10 40 40>
      constraintSetup {angularLimitsMax=<0 5 15>, angularLimitsMin=<0 -40 -15>, angularTarget=<1 0 0>}
      localJointOffset=<5 0 0>
     }]
    relativeSpaceBone {boneName=backpack_root}
   }
   n85 = "Constraint - base_feather_right (PPPPPPPP)"!(
    ComponentPose=@n84
    Alpha=1
    ConstraintWeights_0=0.2829
    ConstraintWeights_1:float=0.2487
    ConstraintWeights_2:float=0.2049
    ConstraintWeights_3:float=0.1451
    ConstraintWeights_4:float=0.0599
    ConstraintWeights_5:float=0.044
    ConstraintWeights_6:float=0.0085
    ConstraintWeights_7:float=0.006
   ) -> (Pose) {
    alphaBoolBlend {blendOption=Linear}
    boneToModify {boneName=base_feather_right}
    constraintSetup [
     {targetBone {boneName=dyn_cape_03}}
     {targetBone {boneName=dyn_cape_r_03}}
     {targetBone {boneName=dyn_cape_r_02}}
     {targetBone {boneName=dyn_cape_02}}
     {targetBone {boneName=dyn_cape_r_04}}
     {targetBone {boneName=dyn_cape_r_01}}
     {targetBone {boneName=dyn_cape_01}}
     {targetBone {boneName=dyn_cape_04}}
    ]
    constraintWeights [1,1,1,1,1,1,1,1]
   }
   n86 = "Constraint - base_feather_mid (PPPPPPPP)"!(
    ComponentPose=@n85
    Alpha=1
    ConstraintWeights_0=0.4567
    ConstraintWeights_1:float=0.3133
    ConstraintWeights_2:float=0.0675
    ConstraintWeights_3:float=0.0494
    ConstraintWeights_4:float=0.0382
    ConstraintWeights_5:float=0.0333
    ConstraintWeights_6:float=0.0264
    ConstraintWeights_7:float=0.0152
   ) -> (Pose) {
    alphaBoolBlend {blendOption=Linear}
    boneToModify {boneName=base_feather_mid}
    constraintSetup [
     {targetBone {boneName=dyn_cape_04}}
     {targetBone {boneName=dyn_cape_03}}
     {targetBone {boneName=dyn_cape_l_04}}
     {targetBone {boneName=dyn_cape_l_03}}
     {targetBone {boneName=dyn_cape_r_04}}
     {targetBone {boneName=dyn_cape_r_03}}
     {targetBone {boneName=dyn_cape_05}}
     {targetBone {boneName=dyn_cape_02}}
    ]
    constraintWeights [1,1,1,1,1,1,1,1]
   }
   n87 = "Constraint - base_feather_left (PPPPPPPP)"!(
    ComponentPose=@n86
    Alpha=1
    ConstraintWeights_0=0.2628
    ConstraintWeights_1:float=0.2207
    ConstraintWeights_2:float=0.1879
    ConstraintWeights_3:float=0.1705
    ConstraintWeights_4:float=0.082
    ConstraintWeights_5:float=0.0378
    ConstraintWeights_6:float=0.0311
    ConstraintWeights_7:float=0.0073
   ) -> (Pose) {
    alphaBoolBlend {blendOption=Linear}
    boneToModify {boneName=base_feather_left}
    constraintSetup [
     {targetBone {boneName=dyn_cape_l_04}}
     {targetBone {boneName=dyn_cape_l_05}}
     {targetBone {boneName=dyn_cape_04}}
     {targetBone {boneName=dyn_cape_05}}
     {targetBone {boneName=dyn_cape_l_03}}
     {targetBone {boneName=dyn_cape_03}}
     {targetBone {boneName=dyn_cape_l_06}}
     {targetBone {boneName=dyn_cape_l_02}}
    ]
    constraintWeights [1,1,1,1,1,1,1,1]
   }
   Save cached pose 'animDyn Cache'(Pose=Component To Local(ComponentPose=@n87)) {cacheName=animDyn Cache}
   n88 = Get GravityOverride()
   n89 = Get bIsCrouching()
   n90 = Get Sim Space Settings() -> (Sim Space Settings:Sim Space Settings Structure)
   n91 = RigidBody!(
    ComponentPose=Local To Component(LocalPose=Use cached pose 'Start_Cache'())
    bAlphaBoolEnabled:bool=@n89
    OverrideWorldGravity:Vector=@n88
    SimSpaceSettings:Sim Space Settings Structure=@n90
   ) -> (Pose) {
    alphaInputType=Bool
    bClampLinearTranslationLimitToRefPose=true
    bOverrideWorldGravity=true
    bUseExternalClothCollision=true
    overridePhysicsAsset=&/M_MED_MediCrow_Pack_PhysAsset
    simulationSpace=BaseBoneSpace
    alphaBoolBlend {blendInTime=3, blendOption=Linear, blendOutTime=3}
    baseBoneRef {boneName=backpack_root}
    componentAppliedLinearAccClamp=<5000 5000 5000>
    componentLinearAccScale {z=0.6}
    componentLinearVelScale {z=0.6}
    overrideWorldGravity {z=-980}
    simSpaceSettings {
     maxAngularAcceleration=500
     maxAngularVelocity=3
     maxLinearAcceleration=700
     maxLinearVelocity=500
     velocityScaleZ=0
     worldAlpha=1
     externalLinearDragV=<1 1>
    }
   }
   n92 = RigidBody(ComponentPose=@n91, bAlphaBoolEnabled=NOT Boolean(A=@n89), OverrideWorldGravity=@n88, SimSpaceSettings=@n90) {
    alphaInputType=Bool
    bClampLinearTranslationLimitToRefPose=true
    bOverrideWorldGravity=true
    bUseExternalClothCollision=true
    overridePhysicsAsset=&/M_MED_MediCrow_Pack_PhysAsset_notcrouch
    simulationSpace=BaseBoneSpace
    alphaBoolBlend {blendInTime=5, blendOption=Linear, blendOutTime=5}
    baseBoneRef {boneName=backpack_root}
    componentAppliedLinearAccClamp=<5000 5000 5000>
    componentLinearAccScale {z=0.6}
    componentLinearVelScale {z=0.6}
    overrideWorldGravity {z=-980}
    simSpaceSettings {
     maxAngularAcceleration=500
     maxAngularVelocity=3
     maxLinearAcceleration=700
     maxLinearVelocity=500
     velocityScaleZ=0
     worldAlpha=1
     externalLinearDragV=<1 1>
    }
   }
   n93 = "Constraint - base_feather_right (PPPPPPPP)"(
    ComponentPose=@n92
    Alpha=1
    ConstraintWeights_0=0.2829
    ConstraintWeights_1=0.2487
    ConstraintWeights_2=0.2049
    ConstraintWeights_3=0.1451
    ConstraintWeights_4=0.0599
    ConstraintWeights_5=0.044
    ConstraintWeights_6=0.0085
    ConstraintWeights_7=0.006
   ) {
    alphaBoolBlend {blendOption=Linear}
    boneToModify {boneName=base_feather_right}
    constraintSetup [
     {targetBone {boneName=dyn_cape_03}}
     {targetBone {boneName=dyn_cape_r_03}}
     {targetBone {boneName=dyn_cape_r_02}}
     {targetBone {boneName=dyn_cape_02}}
     {targetBone {boneName=dyn_cape_r_04}}
     {targetBone {boneName=dyn_cape_r_01}}
     {targetBone {boneName=dyn_cape_01}}
     {targetBone {boneName=dyn_cape_04}}
    ]
    constraintWeights [1,1,1,1,1,1,1,1]
   }
   n94 = "Constraint - base_feather_mid (PPPPPPPP)"(
    ComponentPose=@n93
    Alpha=1
    ConstraintWeights_0=0.4567
    ConstraintWeights_1=0.3133
    ConstraintWeights_2=0.0675
    ConstraintWeights_3=0.0494
    ConstraintWeights_4=0.0382
    ConstraintWeights_5=0.0333
    ConstraintWeights_6=0.0264
    ConstraintWeights_7=0.0152
   ) {
    alphaBoolBlend {blendOption=Linear}
    boneToModify {boneName=base_feather_mid}
    constraintSetup [
     {targetBone {boneName=dyn_cape_04}}
     {targetBone {boneName=dyn_cape_03}}
     {targetBone {boneName=dyn_cape_l_04}}
     {targetBone {boneName=dyn_cape_l_03}}
     {targetBone {boneName=dyn_cape_r_04}}
     {targetBone {boneName=dyn_cape_r_03}}
     {targetBone {boneName=dyn_cape_05}}
     {targetBone {boneName=dyn_cape_02}}
    ]
    constraintWeights [1,1,1,1,1,1,1,1]
   }
   n95 = "Constraint - base_feather_left (PPPPPPPP)"(
    ComponentPose=@n94
    Alpha=1
    ConstraintWeights_0=0.2628
    ConstraintWeights_1=0.2207
    ConstraintWeights_2=0.1879
    ConstraintWeights_3=0.1705
    ConstraintWeights_4=0.082
    ConstraintWeights_5=0.0378
    ConstraintWeights_6=0.0311
    ConstraintWeights_7=0.0073
   ) {
    alphaBoolBlend {blendOption=Linear}
    boneToModify {boneName=base_feather_left}
    constraintSetup [
     {targetBone {boneName=dyn_cape_l_04}}
     {targetBone {boneName=dyn_cape_l_05}}
     {targetBone {boneName=dyn_cape_04}}
     {targetBone {boneName=dyn_cape_05}}
     {targetBone {boneName=dyn_cape_l_03}}
     {targetBone {boneName=dyn_cape_03}}
     {targetBone {boneName=dyn_cape_l_06}}
     {targetBone {boneName=dyn_cape_l_02}}
    ]
    constraintWeights [1,1,1,1,1,1,1,1]
   }
   n96 = Anim Dynamics - Bone: dyn_feather_mid!(ComponentPose=@n95, Alpha=1, GravityOverride:Vector="0,0,0") -> (Pose) {
    angularDampingOverride=0.999
    angularSpringConstant=2000
    bAngularSpring=true
    bOverrideAngularDamping=true
    bUseGravityOverride=true
    bUsePlanarLimit=false
    simulationSpace=RootRelative
    alphaBoolBlend {blendOption=Linear}
    boundBone {boneName=dyn_feather_mid}
    componentAppliedLinearAccClamp=<3000 3000 3000>
    componentLinearAccScale=<0.3 0.3 0.3>
    componentLinearVelScale=<0.3 0.3 0.3>
    physicsBodyDefinitions [{
      boundBone {boneName=dyn_feather_mid}
      boxExtents=<8 8 8>
      constraintSetup {angularLimitsMax=<1.5 0 10>, angularLimitsMin=<-1.5 -50 -10>, angularTarget=<1 0 0>}
      localJointOffset=<3 0 0>
     }]
    relativeSpaceBone {boneName=base_feather_mid}
   }
   n97 = Anim Dynamics - Bone: dyn_feather_left!(ComponentPose=@n96, Alpha=1, GravityOverride:Vector="0,0,0") -> (Pose) {
    angularDampingOverride=0.999
    angularSpringConstant=2000
    bAngularSpring=true
    bOverrideAngularDamping=true
    bUseGravityOverride=true
    bUsePlanarLimit=false
    simulationSpace=RootRelative
    alphaBoolBlend {blendOption=Linear}
    boundBone {boneName=dyn_feather_left}
    componentAppliedLinearAccClamp=<3000 3000 3000>
    componentLinearAccScale=<0.3 0.3 0.3>
    componentLinearVelScale=<0.3 0.3 0.3>
    physicsBodyDefinitions [{
      boundBone {boneName=dyn_feather_left}
      boxExtents=<8 8 8>
      constraintSetup {angularLimitsMax=<1.5 0 10>, angularLimitsMin=<-1.5 -30 -10>, angularTarget=<1 0 0>}
      localJointOffset=<3 0 0>
     }]
   }
   n98 = Anim Dynamics - Bone: dyn_feather_right!(ComponentPose=@n97, Alpha=1, GravityOverride:Vector="0,0,0") -> (Pose) {
    angularDampingOverride=0.999
    angularSpringConstant=2000
    bAngularSpring=true
    bOverrideAngularDamping=true
    bUseGravityOverride=true
    bUsePlanarLimit=false
    bUseSphericalLimits=true
    simulationSpace=RootRelative
    alphaBoolBlend {blendOption=Linear}
    boundBone {boneName=dyn_feather_right}
    componentAppliedLinearAccClamp=<3000 3000 3000>
    componentLinearAccScale=<0.3 0.3 0.3>
    componentLinearVelScale=<0.3 0.3 0.3>
    physicsBodyDefinitions [{
      boundBone {boneName=dyn_feather_right}
      boxExtents=<10 10 10>
      constraintSetup {angularLimitsMax=<1.5 0 5>, angularLimitsMin=<-1.5 -10 -5>, angularTarget=<1 0 0>}
      localJointOffset=<3 0 0>
     }]
    sphericalLimits [{limitRadius=12.5308, drivingBone {boneName=base_feather_right}, sphereLocalOffset=<8.4539 0.2767 -24.7373>}]
   }
   n99 = RigidBody!(
    ComponentPose=@n98
    Alpha:float
    OverrideWorldGravity="0,0,-800"
    SimSpaceSettings=(WorldAlpha=1,VelocityScaleZ=1,DampingAlpha=1,MaxLinearVelocity=300,MaxAngularVelocity=3,MaxLinearAcceleration=1000,MaxAngularAcceleration=5,ExternalLinearDragV=(X=3,Y=7,Z=0),ExternalLinearVelocity=(X=0,Y=0,Z=0),ExternalAngularVelocity=(X=0,Y=0,Z=0))
   ) -> (Pose) {
    bOverrideWorldGravity=true
    overridePhysicsAsset=&/M_MED_MediCrow_Pack_Chain_Physics_for_notcloth
    simulationSpace=BaseBoneSpace
    alphaBoolBlend {blendOption=Linear}
    baseBoneRef {boneName=hider}
    overrideWorldGravity {z=-800}
    simSpaceSettings {maxAngularAcceleration=5, maxAngularVelocity=3, maxLinearAcceleration=1000, maxLinearVelocity=300, worldAlpha=1, externalLinearDragV=<3 7>}
   }
   n100 = RigidBody(
    ComponentPose=Local To Component(LocalPose=Use cached pose 'Start_Cache'())
    OverrideWorldGravity="0,0,-800"
    SimSpaceSettings=(WorldAlpha=1,VelocityScaleZ=1,DampingAlpha=1,MaxLinearVelocity=300,MaxAngularVelocity=3,MaxLinearAcceleration=1000,MaxAngularAcceleration=5,ExternalLinearDragV=(X=3,Y=7,Z=0),ExternalLinearVelocity=(X=0,Y=0,Z=0),ExternalAngularVelocity=(X=0,Y=0,Z=0))
   ) {
    bOverrideWorldGravity=true
    overridePhysicsAsset=&/M_MED_MediCrow_Pack_Chain_Physics
    simulationSpace=BaseBoneSpace
    alphaBoolBlend {blendOption=Linear}
    baseBoneRef {boneName=hider}
    overrideWorldGravity {z=-800}
    simSpaceSettings {maxAngularAcceleration=5, maxAngularVelocity=3, maxLinearAcceleration=1000, maxLinearVelocity=300, worldAlpha=1, externalLinearDragV=<3 7>}
   }
   Save cached pose 'RbanCache'(Pose=Blend Poses by bool(
     BlendPose_0=Component To Local(ComponentPose=@n100)
     BlendPose_1=Component To Local(ComponentPose=@n99)
     BlendTime_0=0.2
     BlendTime_1=0.5
     bActiveValue=Get bClothEnabled()
    ) {blendType=Linear}) {cacheName=RbanCache}
  ]}
 EventGraph {nodes [
   n1 = Gravity Override State Function DEPRECATED(
    CharacterPartAnimInstance=Self-Reference()
    GameTimeInSeconds=GetGameTimeInSeconds()
    GravityOverrideParams=Get Gravity Override Params_1() -> (Gravity Override Params_1:Gravity Override Params Struct Structure)
    RBAN Params=Get RBAN Params_0() -> (RBAN Params_0:RBANParams Struct Structure)
   )
   > Set Gravity Override_metspine(Gravity Override_metspine:Vector=@n1) -> (Output_Get:Vector)
   > Set RBAN Settings Out Spine(RBAN Settings Out Spine:RBANSetting Struct Structure=@n1.RBAN SettingsOut) -> (Output_Get:RBANSetting Struct Structure)
   n2 = Cloth Controller Function DEPRECATED!(
    execute=@n11.then
    CharacterPartAnimInstance:Custom Character Part Anim Instance Object Reference=Self-Reference()
    ClothingAssetName:String=M_MED_MediCrow_Pack_Clothing_0
    Cloth Paramaters:Cloth Params Struct Structure=Get Cloth Paramaters() -> (Cloth Paramaters:Cloth Params Struct Structure)
    Cloth Core Settings:Cloth Core Setting Struct Structure=Get Cloth Core Settings() -> (Cloth Core Settings:Cloth Core Setting Struct Structure)
    ClothUpAxis:Vector="0,0,1"
    OnGround_Moving_SpeedLimit:double=2
    OnGround_Moving_AngularVel_Max:double=0.05
    OnGround_Moving_FictitiousVel_Max:double=0.6
    LinVelScale_FwdBwd_Clamp:double=0
    LinVelScale_UpDwn_Clamp:double=0.7
    DeltaTime:double=0
    GameTimeInSeconds:double=GetGameTimeInSeconds()
    WindGustsParams:Wind Gust Params Struct Structure=Get Wind Gust Params() -> (Wind Gust Params:Wind Gust Params Struct Structure)
    WindGustLocalSettings:Wind Gust Local Setting Struct Structure=Get Gust Local Settings() -> (Gust Local Settings:Wind Gust Local Setting Struct Structure)
   ) -> (
    then
    CurrentStateOut:String
    Cloth Settings Out:Cloth Setting Struct Structure
    Cloth Core Settings Out:Cloth Core Setting Struct Structure
    WindGustSettings Out:Wind Gust Setting Struct Structure
    WindGustLocalSettings Out:Wind Gust Local Setting Struct Structure
    GustStrength - WindVelocityOut:Vector
    LinearVelVector-DELETE LATER:Vector
   )
   n3 = GetOwnerSkeletalMeshComponent()
   n4 = DrawVectorFromJointFunction!(
    execute
    SkelMeshComponent:Scene Component Object Reference=@n3
    JointName:Name=pelvis
    LineEnd:Vector=Break Cloth Setting Struct!(ClothSettingStruct:"Cloth Setting Struct Structure (by ref)"=@n2.Cloth Settings Out) -> (
     LinearVelocityScale_7_#1:Vector=~
     AngularVelocityScale_6_#2:float=~
     FictitiousAngularScale_10_#3:float=~
     GravityOverride_15_#4:Vector=~
     DragCoefficient_23_#5:Vector 2D Structure
     LiftCoefficient_24_#6:Vector 2D Structure
     JointName_27_#7:Name
    ) {is=title K2Node_BreakStruct}.GravityOverride_15_#4
    JointOffset:Vector="0, 0, 0"
    LineColor:Linear Color Structure=(R=1,G=0,B=0,A=1)
    Thickness:double=2
    Foreground:EDrawDebugSceneDepthPriorityGroup Enum=Foreground
   ) -> (then)
   DrawJointDebugFunction!(
    execute=[@n4.then, @n7.then]
    SkelMeshComponent:Scene Component Object Reference=@n3
    JointOffset:Vector="0, 0, 0"
    JointName:Array of Names=Make Array("[0]":Name=root) -> (Array:Array of Names)
    Scale:double=10
    Duration:double=0
    Thickness:double=0.7
    DebugDraw:bool=true
    Foreground:EDrawDebugSceneDepthPriorityGroup Enum=Foreground
   ) -> (then)
   n5 = Gravity Override State Function DEPRECATED(
    execute=@n11.else
    CharacterPartAnimInstance=Self-Reference()
    GameTimeInSeconds=GetGameTimeInSeconds()
    GravityOverrideParams=Get Gravity Override Params()
   )
   > n6 = Set GravityOverride(GravityOverride=@n5)
   n7 = DrawVectorFromJointFunction(
    SkelMeshComponent=@n3
    JointName=spine_05
    LineEnd=@n6
    JointOffset="0, 0, 0"
    LineColor=(R=1,G=0,B=0,A=1)
    Thickness=2
    Foreground=Foreground
   )
   Event BlueprintUpdateAnimation()
   > n8 = Sequence!(execute) -> (then_0, then_1, then_2:Exec, then_3:Exec, then_4:Exec, then_5:Exec, then_6:Exec, then_7:Exec, then_8:Exec, then_9:Exec)
   n9 = GetOwnerSkeletalMeshComponent()
   > Set spin02(spin02:Transform=GetSocketTransform(self=@n9, InSocketName=spine_02, TransformSpace=RTS_World)) -> (Output_Get:Transform)
   > Set spin03(spin03:Transform=GetSocketTransform(self=@n9, InSocketName=spine_03, TransformSpace=RTS_World)) -> (Output_Get:Transform)
   > Set spin04(spin04:Transform=GetSocketTransform(self=@n9, InSocketName=spine_04, TransformSpace=RTS_World)) -> (Output_Get:Transform)
   > n10 = Gravity Override State Function DEPRECATED(
    CharacterPartAnimInstance=Self-Reference()
    GameTimeInSeconds=GetGameTimeInSeconds()
    GravityOverrideParams=Get Gravity Override Params_0() -> (Gravity Override Params_0:Gravity Override Params Struct Structure)
    RBAN Params=Get RBAN Params()
   )
   > Set Gravity Override_chain(Gravity Override_chain:Vector=@n10) -> (Output_Get:Vector)
   > Set RBAN Settings Out_chain(RBAN Settings Out_chain:RBANSetting Struct Structure=@n10.RBAN SettingsOut) -> (Output_Get:RBANSetting Struct Structure)
   > n11 = Branch(Condition=Get bClothEnabled())
   n12 = GetClothingSimulationInteractor!(self:Skeletal Mesh Component Object Reference=GetOwningComponent(), ClothingAssetName:Name=None) -> (ReturnValue:Clothing Simulation Interactor Object Reference)
   Is Valid(exec=@n8.then_1, InputObject=@n12)
   n13 = MakeLiteralString(Value=M_MED_MediCrow_Pack_Clothing_0)
   > n14 = Cast To ChaosClothingInteractor!(
    execute
    Object=GetClothingInteractor!(self:Clothing Simulation Interactor Object Reference=@n12, ClothingAssetName:Name=String To Name(InString=@n13)) -> (ReturnValue:Clothing Interactor Object Reference)
   ) -> (then, CastFailed, AsChaos Clothing Interactor:Chaos Clothing Interactor Object Reference) {is=title K2Node_DynamicCast}
   > Set ChaosClothingInteractor(ChaosClothingInteractor:Chaos Clothing Interactor Object Reference=@n14) -> (Output_Get:Chaos Clothing Interactor Object Reference)
   n15 = GetPartSkeletalMeshComponent()
   Set TeleportDistanceThreshold!(execute=@n8.then_2, TeleportDistanceThreshold:float=300, self:Skeletal Mesh Component Object Reference=@n15) -> (then, Output_Get:float)
   > Set TeleportRotationThreshold!(execute, TeleportRotationThreshold:float=25, self:Skeletal Mesh Component Object Reference=@n15) -> (then, Output_Get:float)
   n16 = Branch(execute=@n8.then_3, Condition=Get bIsInFrontEnd())
   > Set BwdMovement(BwdMovement:double=0.5) -> (Output_Get:double)
   n17 = PlayerMovement(execute=@n16.else, AnimInstance=Self-Reference(), XDotProductCoefficient=1, YDotProductCoefficient=1, ZDotProductCoefficient=1)
   > n18 = Branch(Condition=Get bIsDBNOCarried())
   > Set BwdMovement(BwdMovement=0.5)
   n19 = Branch(execute=@n18.else, Condition=AND Boolean(A=Get bIsTargeting(), B=Get bIsCrouching()))
   n20 = MapRangeClamped(
    Value=float + float(A="Absolute (Float)"(A=@n17), B=MapRangeClamped(Value=@n17.YMovement, InRangeA=-1, InRangeB=1, OutRangeA=0.3)) {is=title K2Node_CommutativeAssociativeBinaryOperator}
    InRangeA=0
    InRangeB=1.65
   )
   > Set BwdMovement(BwdMovement=float * float(A=@n20, B=0.25) {is=title K2Node_PromotableOperator})
   Set BwdMovement(execute=@n19.else, BwdMovement=@n20)
   n21 = Branch(execute=@n8.then_4, Condition=Get bIsSkyDiving())
   > n22 = Branch(Condition="float > float"(A=Get PawnVelocityZ(), B=50) {is=title K2Node_CallFunction})
   > Set BwdMovement(BwdMovement=float * float(A=Get BwdMovement(), B=0) {is=title K2Node_CommutativeAssociativeBinaryOperator})
   Set BwdMovement(execute=[@n21.else, @n22.else], BwdMovement=float * float(A=Get BwdMovement(), B=1))
   Set RbanOrCloth(execute=@n8.then_5, RbanOrCloth:bool=OR Boolean(A=Get bRigidBodyEnabled(), B=Get bClothEnabled())) -> (Output_Get:bool)
   n23 = Branch(execute=@n8.then_6, Condition=Get bIsParachuteOpened())
   > Set RotationBP(RotationBP:Rotator="0,-45,0") -> (Output_Get:Rotator)
   Set RotationBP(execute=@n23.else, RotationBP="0,-25,0")
   n24 = Sequence!(execute=@n8.then_8) -> (then_0, then_1)
   > n25 = Branch(Condition=Get bIsInFrontEnd())
   > n26 = Cast To FortPlayerPawn(Object=GetOwningActor())
   > n27 = Branch(Condition=Get bIsSkyDiving())
   n28 = GetSkeletalMeshForPartType(self=@n26)
   > n29 = Get Joint Location Relative to Bone!(
    execute
    Target Joint:Name=pelvis
    Relative Joint:Name=pelvis
    Offset:Vector="0,-40,0"
    SkelMeshComponent:Mesh Component Object Reference=@n28
    Debug Draw:bool=false
   ) -> (then, Joint Location Vector:Vector) {is=title K2Node_CallFunction}
   > Set BackpackRoot_RotationOffset(BackpackRoot_RotationOffset:Vector=@n29) -> (Output_Get:Vector)
   n30 = Get Joint Location Relative to Bone(execute=@n27.else, Target Joint=pelvis, Relative Joint=pelvis, Offset="0,-20,0", SkelMeshComponent=@n28, Debug Draw=false)
   > Set BackpackRoot_RotationOffset(BackpackRoot_RotationOffset=@n30)
   n31 = Cast To FortPlayerPawn(execute=@n25.else, Object=GetOwningActor())
   > Set BackpackRoot_RotationOffset(BackpackRoot_RotationOffset="0,-15,0")
   > n32 = Get Joint Location Relative to Bone(Target Joint=cape, Relative Joint=cape, Offset="0,0,-50", SkelMeshComponent=GetSkeletalMeshForPartType(self=@n31), Debug Draw=false)
   > n33 = DistanceBetweenTwoVectors!(
    execute
    vector1:Vector=@n32
    Vector2:Vector=GetSocketLocation!(self:Scene Component Object Reference=GetOwnerSkeletalMeshComponent(), InSocketName:Name=pelvis) -> (ReturnValue:Vector)
   ) -> (then, Distance:double)
   n34 = BreakVector(InVec=Get BackpackRoot_RotationOffset() -> (BackpackRoot_RotationOffset:Vector))
   n35 = float / float(A=@n33, B=17)
   > Set BackpackRoot_RotationOffset(BackpackRoot_RotationOffset=MakeVector(X=@n34, Y=float * float(A=@n34.Y, B=@n35) {is=title K2Node_PromotableOperator}, Z=@n34.Z))
   n36 = Branch(execute=@n24.then_1, Condition=Get bIsInFrontEnd())
   > Set Cape_LookAt_Target(Cape_LookAt_Target:Vector=Get BackpackRoot_RotationOffset()) -> (Output_Get:Vector)
   n37 = Get Joint Location Relative to Bone(
    execute=@n36.else
    Target Joint=pelvis
    Relative Joint=pelvis
    Offset=Get BackpackRoot_RotationOffset()
    SkelMeshComponent=GetOwnerSkeletalMeshComponent()
    Debug Draw=false
   )
   > Set Cape_LookAt_Target(Cape_LookAt_Target=@n37)
   n38 = Cast To FortPlayerPawn(execute=@n8.then_9, Object=TryGetPawnOwner())
   > n39 = Branch(Condition=Get bIsOnGrindRail())
   > n40 = Branch(Condition=NOT Boolean(A=IsWeaponHolstered!(self:Fort Pawn Object Reference=@n38) -> (ReturnValue:bool)))
   > n41 = Branch(Condition=Get bIsTargeting())
   > Set Sim Space Settings(Sim Space Settings:Sim Space Settings Structure=Get Sim Space Settings_GrindTarget() -> (Sim Space Settings_GrindTarget:Sim Space Settings Structure)) -> (Output_Get:Sim Space Settings Structure)
   Set Sim Space Settings(
    execute=@n41.else
    Sim Space Settings=Get Sim Space Settings_GrindingWeaponActive() -> (Sim Space Settings_GrindingWeaponActive:Sim Space Settings Structure)
   )
   Set Sim Space Settings(execute=[@n39.else, @n40.else], Sim Space Settings=Get Sim Space Settings_Base() -> (Sim Space Settings_Base:Sim Space Settings Structure))
   Event BlueprintInitializeAnimation()
   > Set bResetAfterTeleport!(execute, bResetAfterTeleport:bool=false, self:Skeletal Mesh Component Object Reference=@n15) -> (then, Output_Get:bool)
   Event OnEndFalling()
   > Branch(Condition=Get resetClothLandingFromZilpine() -> (resetClothLandingFromZilpine:bool))
   n42 = Get ChaosClothingInteractor() -> (ChaosClothingInteractor:Chaos Clothing Interactor Object Reference)
   > ResetAndTeleport!(execute, self:Chaos Clothing Interactor Object Reference=@n42, bReset:bool=true, bTeleport:bool=true) -> (then)
   > Set resetClothLandingFromZilpine(resetClothLandingFromZilpine:bool=false) -> (Output_Get:bool)
   Event OnBeginSkydiving()
   > Set SkydivingAlpha(SkydivingAlpha:Rotator="65,0,0") -> (Output_Get:Rotator)
   n43 = Event BlueprintProcessAnimInputEvent()
   > n44 = Branch(Condition=IsNamed(self=@n43.AnimInputEvent, InEventName=aie.Player.Zipline.Entered))
   > Delay(Duration=0.1)
   > ResetAndTeleport(self=@n42, bReset=true, bTeleport=true)
   Branch(execute=@n44.else, Condition=IsNamed(self=@n43.AnimInputEvent, InEventName=aie.Player.Zipline.Exited))
   > Set resetClothLandingFromZilpine(resetClothLandingFromZilpine=true)
   Event OnEndSkydiving()
   > Set SkydivingAlpha(SkydivingAlpha="10,0,0")
   DrawVectorFromJointFunction(
    SkelMeshComponent=@n3
    JointName=pelvis
    LineEnd=@n2.GustStrength - WindVelocityOut
    JointOffset="0, 0, 0"
    LineColor=(R=0.9957,G=1,B=0,A=1)
    Thickness=2
    Foreground=Foreground
   )
   DrawJointDebugFunction(
    SkelMeshComponent=@n15
    JointOffset="0, 0, 0"
    JointName=Make Array("[0]"=root)
    Scale=10
    Duration=0
    Thickness=0.7
    DebugDraw=true
    Foreground=Foreground
   )
   Debug Cloth Teleport Threshold Function!(
    execute
    MinDistance:double=80
    MinRotation:double=25
    SkelMeshComponent:Scene Component Object Reference=@n15
    PrintToScreen:bool=true
    BoneName:Name=root
    InVector:"Vector (by ref)"=Get TeleportDistanceThreshold_Local() -> (TeleportDistanceThreshold_Local:Vector)
    PawnDeltaRotation:Rotator=Get PawnDeltaRotation() -> (PawnDeltaRotation:Rotator)
    PrintToScreenDuration:double=6
   ) -> (then, Distance:double, Rotation:double)
   GetSocketTransform(InSocketName=spine_05, TransformSpace=RTS_World)
   GetSocketTransform(InSocketName=pelvis, TransformSpace=RTS_World)
   GetSocketTransform(InSocketName=spine_01, TransformSpace=RTS_World)
   TryGetPawnOwner()
   Set Gust Local Settings(execute=@n2.then, Gust Local Settings:Wind Gust Local Setting Struct Structure=@n2.WindGustLocalSettings Out) -> (Output_Get:Wind Gust Local Setting Struct Structure)
  ]}
 OnInitialUpdate {nodes [
   n1 = OnInitialUpdate!() -> (then, Context:"Anim Update Context Structure (by ref)", Node:"Anim Node Reference Structure (by ref)") {is=title K2Node_FunctionEntry}
   > n2 = ConvertToRigidBodyAnimNode!(execute, Node:"Anim Node Reference Structure (by ref)"=@n1.Node) -> (Succeeded:Exec, Failed:Exec, ReturnValue:Rigid Body Anim Node Reference Structure)
   > SetOverridePhysicsAsset!(
    execute
    Node:"Rigid Body Anim Node Reference Structure (by ref)"=@n2
    PhysicsAsset:Physics Asset Object Reference=Get Physics Asset() -> (Physics Asset:Physics Asset Object Reference)
   ) -> (then, ReturnValue:Rigid Body Anim Node Reference Structure)
   > Return Node()
  ]}
]
variables [
 BackpackRoot_RotationOffset {type=Vector}
 BodyMesh {type=SkeletalMeshComponentBudgeted}
 BwdMovement {type=number}
 Cape_LookAt_Target {type=Vector}
 ChaosClothingInteractor {type=ChaosClothingInteractor, default=null}
 Cloth Core Settings {
  type=ClothCoreSettingStruct
  default {
   collisionThickness=2
   dampingCoefficient=0.1
   frictionCoefficient=0.1
   localDampingCoefficient=0
   numberIterations=2
   numberSubsteps=1
   selfCollisionThickness=0
   animDriveDamping=<0 1>
   animDriveStiffness=<0 1>
   areaStiffness=<1 1>
   bendingStiffness=<1 1>
   edgeStiffness=<1 1>
   tetherStiffness=<1 1>
  }
 }
 Cloth Paramaters {
  type=ClothParamsStruct
  default {
   dBNO {
    angularVelocityScale=0.6
    fictitiousAngularScale=0.1
    jointName=pelvis
    dragCoefficient=<0.2 1>
    gravityOverride=<-980 400 0>
    liftCoefficient=<0 0.5>
    linearVelocityScale=<0.3 0.3 0.1>
   }
   dBNO_Carried {
    angularVelocityScale=0.6
    fictitiousAngularScale=0.1
    jointName=pelvis
    dragCoefficient=<0.2 1>
    gravityOverride=<-980 -200 0>
    liftCoefficient=<0 0.5>
    linearVelocityScale=<0.3 0.3 0.1>
   }
   dBNO_Impulse {
    angularVelocityScale=0
    fictitiousAngularScale=0
    jointName=pelvis
    dragCoefficient=<0 0.2>
    gravityOverride=<-3000 -2000 0>
    liftCoefficient=<0 0.1>
    linearVelocityScale=<0.3 0.1 0.3>
   }
   dBNO_RollBwd {
    angularVelocityScale=0
    fictitiousAngularScale=0
    jointName=pelvis
    dragCoefficient=<0 0.2>
    gravityOverride=<-2000 0 0>
    liftCoefficient=<0 0.1>
    linearVelocityScale=<0.3 0.1 0.3>
   }
   dBNO_RollFwd {
    angularVelocityScale=0
    fictitiousAngularScale=0
    jointName=pelvis
    dragCoefficient=<0 0.2>
    gravityOverride=<-980 0 0>
    liftCoefficient=<0 0.1>
    linearVelocityScale=<0.3 0.1 0.3>
   }
   dBNO_Tumbling {
    angularVelocityScale=0
    fictitiousAngularScale=0
    jointName=pelvis
    dragCoefficient=<0 0.2>
    gravityOverride=<-2500 -300 0>
    liftCoefficient=<0 0.1>
    linearVelocityScale=<0.3 0.1 0.3>
   }
   emoteOrMelee {
    angularVelocityScale=0.6
    fictitiousAngularScale=0.1
    jointName=root
    dragCoefficient=<0 0.3>
    gravityOverride=<0 0 -980>
    liftCoefficient=<0 0.2>
    linearVelocityScale=<0.6 0.6 0.5>
   }
   front_End {
    angularVelocityScale=0.05
    fictitiousAngularScale=0.3
    jointName=root
    dragCoefficient=<0.2 0.2>
    gravityOverride=<0 0 -980>
    liftCoefficient=<0.1 0.1>
    linearVelocityScale=<0.1 0.1 0.1>
   }
   onGround_Crouch_Moving {
    angularVelocityScale=0.1
    fictitiousAngularScale=0.3
    jointName=root
    dragCoefficient=<2 2>
    gravityOverride=<0 0 -980>
    liftCoefficient=<0.5 0.5>
    linearVelocityScale=<0.6 0.6 0.5>
   }
   onGround_Falling {
    angularVelocityScale=0.6
    fictitiousAngularScale=0
    jointName=root
    dragCoefficient=<0 0.2>
    gravityOverride=<0 -500 -980>
    liftCoefficient=<0 0.1>
    linearVelocityScale=<2 1 0.3>
   }
   onGround_Moving {
    angularVelocityScale=0.6
    fictitiousAngularScale=0.1
    jointName=root
    dragCoefficient=<0.2 3>
    gravityOverride=<0 -1000 -980>
    liftCoefficient=<0 0.3>
    linearVelocityScale=<0.7 0.7 0.3>
   }
   onGround_Standing {
    angularVelocityScale=0.05
    fictitiousAngularScale=1
    jointName=root
    dragCoefficient=<0.2 2.5>
    gravityOverride=<0 0 -980>
    liftCoefficient=<0 0.3>
    linearVelocityScale=<0.2 0.2 0.3>
   }
   rail_Grinding {
    angularVelocityScale=0.2
    fictitiousAngularScale=0
    jointName=root
    dragCoefficient=<0.1 0.2>
    gravityOverride=<50 -500 -400>
    liftCoefficient=<0 0.5>
    linearVelocityScale=<0.1 0.1 0.1>
   }
   riding_Moving {
    angularVelocityScale=0.05
    fictitiousAngularScale=1
    jointName=root
    dragCoefficient=<0.5 0.5>
    gravityOverride=<0 0 -980>
    liftCoefficient=<0.5 0.5>
    linearVelocityScale=<0.4 0.6 0.3>
   }
   riding_Moving_Targeting {
    angularVelocityScale=0.05
    fictitiousAngularScale=0.7
    jointName=root
    dragCoefficient=<0.5 0.5>
    gravityOverride=<0 0 -980>
    liftCoefficient=<0.5 0.5>
    linearVelocityScale=<0.2 0.6 0.2>
   }
   riding_Standing {
    angularVelocityScale=0.05
    fictitiousAngularScale=1
    jointName=root
    dragCoefficient=<0.5 0.5>
    gravityOverride=<0 -450 -980>
    liftCoefficient=<0.5 0.5>
    linearVelocityScale=<0.6 0.6 0.3>
   }
   skydive_DivingDown {
    angularVelocityScale=0.1
    fictitiousAngularScale=0
    jointName=pelvis
    dragCoefficient=<0 1>
    gravityOverride=<-450 0 0>
    liftCoefficient=<0 0.1>
    linearVelocityScale=<0.1 0.1 0.1>
   }
   skydive_FloatingDown {
    angularVelocityScale=0.1
    fictitiousAngularScale=0
    jointName=root
    dragCoefficient=<0 2>
    gravityOverride=<0 -800 200>
    liftCoefficient=<0 0.4>
    linearVelocityScale=<0.2 0.2 0.1>
   }
   skydive_GoingUp {
    angularVelocityScale=0.1
    fictitiousAngularScale=0
    jointName=root
    dragCoefficient=<0.1 0.1>
    gravityOverride=<0 -10000 5000>
    liftCoefficient=<0.05 0.05>
    linearVelocityScale=<0.01 0.1 0.1>
   }
   skydive_Parachuting {
    angularVelocityScale=0.6
    fictitiousAngularScale=0.1
    jointName=root
    dragCoefficient=<0 2>
    gravityOverride=<0 0 -980>
    liftCoefficient=<0 0.5>
    linearVelocityScale=<0.2 0.3 0.3>
   }
   skydive_Surfing {
    angularVelocityScale=0.6
    fictitiousAngularScale=0.1
    jointName=root
    dragCoefficient=<1.5 1.5>
    gravityOverride=<0 -600 -500>
    liftCoefficient=<1 1>
    linearVelocityScale=<0.1 0.1 0.3>
   }
   swimming {
    angularVelocityScale=0.6
    fictitiousAngularScale=0
    jointName=root
    dragCoefficient=<2.5 2.5>
    gravityOverride=<0 0 0>
    liftCoefficient=<0.1 0.1>
    linearVelocityScale=<0.6 0.6 0.5>
   }
   tactical_Sprint {
    angularVelocityScale=0.6
    fictitiousAngularScale=0.1
    jointName=root
    dragCoefficient=<2 2>
    gravityOverride=<0 0 -980>
    liftCoefficient=<0.5 0.5>
    linearVelocityScale=<0.6 0.6 0.5>
   }
   ziplining {
    angularVelocityScale=0.6
    fictitiousAngularScale=0
    jointName=root
    dragCoefficient=<0 1.5>
    gravityOverride=<0 0 -980>
    liftCoefficient=<0 0.3>
    linearVelocityScale=<0.4 0.4 0.5>
   }
  }
 }
 Gravity Override Params {type=GravityOverrideParamsStruct, default {rail_Grinding {jointName=spine_05, gravityOverride=<-500 0 -500>}}}
 Gravity Override Params_0 {
  type=GravityOverrideParamsStruct
  default {
   onGround_Moving {gravityOverride=<0 -1000 -800>}
   onGround_Standing {gravityOverride=<0 0 -800>}
   skydive_Parachuting {gravityOverride=<0 0 -700>, windAmplitude=<0 0 0>, windFrequency=<0 0 0>}
   tactical_Sprint {gravityOverride=<0 0 -200>}
  }
 }
 Gravity Override Params_1 {
  type=GravityOverrideParamsStruct
  default {onGround_Moving {gravityOverride=<0 0 0>}, onGround_Standing {gravityOverride=<0 0 0>}, skydive_Parachuting {gravityOverride=<0 -1500 0>}}
 }
 Gravity Override_chain {type=Vector}
 Gravity Override_metspine {type=Vector}
 GravityOverride {type=Vector}
 Gust Local Settings {type=WindGustLocalSettingStruct, default {elapsedTime=0, finalGustStrengthLocal=0, gustDuration=0, gustStrength=0, timeBetweenGusts=0}}
 Physics Asset {
  type=PhysicsAsset
  default=&M_MED_MediCrow_Pack_PhysAsset
  class=~
  bodies=~
 }
 RBAN Params {
  type=RBANParamsStruct
  default {
   dBNO {simSpaceSettings {
     maxAngularAcceleration=1000
     maxAngularVelocity=3
     maxLinearAcceleration=300
     maxLinearVelocity=700
     velocityScaleZ=0
     externalLinearDragV=<0 0.5 0>
    }}
   dBNO_Carried {simSpaceSettings {
     maxAngularAcceleration=1000
     maxAngularVelocity=3
     maxLinearAcceleration=300
     maxLinearVelocity=700
     velocityScaleZ=0
     externalLinearDragV=<0 0.5 0>
    }}
   emoteOrMelee {simSpaceSettings {
     maxAngularAcceleration=1000
     maxAngularVelocity=3
     maxLinearAcceleration=300
     maxLinearVelocity=700
     velocityScaleZ=0
     externalLinearDragV=<0 0.5 0>
    }}
   front_End {simSpaceSettings {
     maxAngularAcceleration=1000
     maxAngularVelocity=3
     maxLinearAcceleration=300
     maxLinearVelocity=700
     velocityScaleZ=0
     worldAlpha=0.3
     externalLinearDragV=<0 0.5 0>
    }}
   onGround_Crouch_Moving {simSpaceSettings {
     maxAngularAcceleration=5
     maxAngularVelocity=3
     maxLinearAcceleration=1000
     maxLinearVelocity=300
     velocityScaleZ=1
     externalLinearDragV=<3 7 0>
    }}
   onGround_Falling {simSpaceSettings {
     maxAngularAcceleration=1000
     maxAngularVelocity=3
     maxLinearAcceleration=300
     maxLinearVelocity=700
     velocityScaleZ=0
     externalLinearDragV=<0 0.5 0>
    }}
   onGround_Moving {
    componentLinearAccScale=<0 0 0>
    componentLinearVelScale=<0 0 0>
    simSpaceSettings {
     maxAngularAcceleration=10
     maxAngularVelocity=5
     maxLinearAcceleration=700
     maxLinearVelocity=300
     velocityScaleZ=0.8
     worldAlpha=0.5
     externalLinearDragV=<6 0 0>
    }
   }
   onGround_Standing {
    componentLinearAccScale=<0 0 0>
    componentLinearVelScale=<0 0 0>
    simSpaceSettings {
     maxAngularAcceleration=5
     maxAngularVelocity=3
     maxLinearAcceleration=1000
     maxLinearVelocity=300
     velocityScaleZ=0.3
     externalLinearDragV=<3 3 0>
    }
   }
   rail_Grinding {simSpaceSettings {
     maxAngularAcceleration=1000
     maxAngularVelocity=3
     maxLinearAcceleration=300
     maxLinearVelocity=700
     velocityScaleZ=0
     worldAlpha=0.1
     externalLinearDragV=<0 0.5 0>
    }}
   riding_Moving {simSpaceSettings {
     maxAngularAcceleration=1000
     maxAngularVelocity=3
     maxLinearAcceleration=300
     maxLinearVelocity=700
     velocityScaleZ=0
     externalLinearDragV=<0 0.5 0>
    }}
   riding_Moving_Targeting {simSpaceSettings {
     maxAngularAcceleration=1000
     maxAngularVelocity=3
     maxLinearAcceleration=300
     maxLinearVelocity=700
     velocityScaleZ=0
     externalLinearDragV=<0 0.5 0>
    }}
   riding_Standing {simSpaceSettings {
     maxAngularAcceleration=1000
     maxAngularVelocity=3
     maxLinearAcceleration=300
     maxLinearVelocity=700
     velocityScaleZ=0
     externalLinearDragV=<0 0.5 0>
    }}
   skydive_DivingDown {simSpaceSettings {
     maxAngularAcceleration=1000
     maxAngularVelocity=3
     maxLinearAcceleration=300
     maxLinearVelocity=700
     velocityScaleZ=0
     externalLinearDragV=<0 0.5 0>
    }}
   skydive_FloatingDown {
    componentLinearAccScale=<0 0.5 0>
    componentLinearVelScale=<0 0.5 0>
    simSpaceSettings {
     maxAngularAcceleration=1000
     maxAngularVelocity=3
     maxLinearAcceleration=300
     maxLinearVelocity=700
     velocityScaleZ=0
     externalLinearDragV=<0 0.5 0>
    }
   }
   skydive_GoingUp {simSpaceSettings {
     maxAngularAcceleration=1000
     maxAngularVelocity=3
     maxLinearAcceleration=300
     maxLinearVelocity=700
     velocityScaleZ=0
     externalLinearDragV=<0 0.5 0>
    }}
   skydive_Parachuting {
    componentLinearAccScale=<0 0.1 0.3>
    componentLinearVelScale=<0 0.1 0.3>
    simSpaceSettings {
     maxAngularAcceleration=1000
     maxAngularVelocity=3
     maxLinearAcceleration=300
     maxLinearVelocity=700
     velocityScaleZ=0
     externalLinearDragV=<0 0.5 0>
    }
   }
   skydive_Surfing {simSpaceSettings {
     maxAngularAcceleration=1000
     maxAngularVelocity=3
     maxLinearAcceleration=300
     maxLinearVelocity=700
     velocityScaleZ=0
     externalLinearDragV=<0 0.5 0>
    }}
   swimming {simSpaceSettings {
     maxAngularAcceleration=1000
     maxAngularVelocity=3
     maxLinearAcceleration=300
     maxLinearVelocity=700
     velocityScaleZ=0
     externalLinearDragV=<0 0.5 0>
    }}
   tactical_Sprint {simSpaceSettings {
     maxAngularAcceleration=1000
     maxAngularVelocity=3
     maxLinearAcceleration=300
     maxLinearVelocity=700
     velocityScaleZ=0
     externalLinearDragV=<0 0.5 0>
    }}
   ziplining {simSpaceSettings {
     maxAngularAcceleration=1000
     maxAngularVelocity=3
     maxLinearAcceleration=300
     maxLinearVelocity=700
     velocityScaleZ=0
     externalLinearDragV=<0 0.5 0>
    }}
  }
 }
 RBAN Params_0 {
  type=RBANParamsStruct
  default {
   dBNO {simSpaceSettings {
     maxAngularAcceleration=1000
     maxAngularVelocity=3
     maxLinearAcceleration=300
     maxLinearVelocity=700
     velocityScaleZ=0
     externalLinearDragV=<0 0.5 0>
    }}
   dBNO_Carried {simSpaceSettings {
     maxAngularAcceleration=1000
     maxAngularVelocity=3
     maxLinearAcceleration=300
     maxLinearVelocity=700
     velocityScaleZ=0
     externalLinearDragV=<0 0.5 0>
    }}
   emoteOrMelee {
    componentLinearAccScale=<0.02 0.5 0>
    componentLinearVelScale=<0.02 0.5 0>
    simSpaceSettings {
     maxAngularAcceleration=1000
     maxAngularVelocity=3
     maxLinearAcceleration=300
     maxLinearVelocity=700
     velocityScaleZ=0
     externalLinearDragV=<0 0.5 0>
    }
   }
   front_End {simSpaceSettings {
     maxAngularAcceleration=1000
     maxAngularVelocity=3
     maxLinearAcceleration=300
     maxLinearVelocity=700
     velocityScaleZ=0
     externalLinearDragV=<0 0.5 0>
    }}
   onGround_Crouch_Moving {
    componentLinearAccScale=<0.3 0.5 0.3>
    componentLinearVelScale=<0.3 0.5 0.3>
    simSpaceSettings {
     maxAngularAcceleration=1000
     maxAngularVelocity=3
     maxLinearAcceleration=300
     maxLinearVelocity=700
     velocityScaleZ=0
     externalLinearDragV=<0 0.5 0>
    }
   }
   onGround_Falling {simSpaceSettings {
     maxAngularAcceleration=1000
     maxAngularVelocity=3
     maxLinearAcceleration=300
     maxLinearVelocity=700
     velocityScaleZ=0
     externalLinearDragV=<0 0.5 0>
    }}
   onGround_Moving {
    componentLinearAccScale=<0.1 0.5 0.1>
    componentLinearVelScale=<0.1 0.5 0.1>
    simSpaceSettings {
     maxAngularAcceleration=1000
     maxAngularVelocity=3
     maxLinearAcceleration=300
     maxLinearVelocity=700
     velocityScaleZ=0
     externalLinearDragV=<0 0.5 0>
    }
   }
   onGround_Standing {
    componentLinearAccScale=<0.02 0.05 0>
    componentLinearVelScale=<0.02 0.05 0>
    simSpaceSettings {
     maxAngularAcceleration=1000
     maxAngularVelocity=3
     maxLinearAcceleration=300
     maxLinearVelocity=700
     velocityScaleZ=0
     externalLinearDragV=<0 0.5 0>
    }
   }
   rail_Grinding {simSpaceSettings {
     maxAngularAcceleration=1000
     maxAngularVelocity=3
     maxLinearAcceleration=300
     maxLinearVelocity=700
     velocityScaleZ=0
     worldAlpha=0.1
     externalLinearDragV=<0 0.5 0>
    }}
   riding_Moving {simSpaceSettings {
     maxAngularAcceleration=1000
     maxAngularVelocity=3
     maxLinearAcceleration=300
     maxLinearVelocity=700
     velocityScaleZ=0
     externalLinearDragV=<0 0.5 0>
    }}
   riding_Moving_Targeting {simSpaceSettings {
     maxAngularAcceleration=1000
     maxAngularVelocity=3
     maxLinearAcceleration=300
     maxLinearVelocity=700
     velocityScaleZ=0
     externalLinearDragV=<0 0.5 0>
    }}
   riding_Standing {simSpaceSettings {
     maxAngularAcceleration=1000
     maxAngularVelocity=3
     maxLinearAcceleration=300
     maxLinearVelocity=700
     velocityScaleZ=0
     externalLinearDragV=<0 0.5 0>
    }}
   skydive_DivingDown {
    componentLinearAccScale=<0.3 0.5 0.3>
    componentLinearVelScale=<0.3 0.5 0.3>
    simSpaceSettings {
     maxAngularAcceleration=1000
     maxAngularVelocity=3
     maxLinearAcceleration=300
     maxLinearVelocity=700
     velocityScaleZ=0
     externalLinearDragV=<0 0.5 0>
    }
   }
   skydive_FloatingDown {
    componentLinearAccScale=<0.3 0.5 0.3>
    componentLinearVelScale=<0.3 0.5 0.3>
    simSpaceSettings {
     maxAngularAcceleration=1000
     maxAngularVelocity=3
     maxLinearAcceleration=300
     maxLinearVelocity=700
     velocityScaleZ=0
     externalLinearDragV=<0 0.5 0>
    }
   }
   skydive_GoingUp {
    componentLinearAccScale=<0.3 0.5 0.3>
    componentLinearVelScale=<0.3 0.5 0.3>
    simSpaceSettings {
     maxAngularAcceleration=1000
     maxAngularVelocity=3
     maxLinearAcceleration=300
     maxLinearVelocity=700
     velocityScaleZ=0
     externalLinearDragV=<0 0.5 0>
    }
   }
   skydive_Parachuting {
    componentLinearAccScale=<0.3 0.5 0.3>
    componentLinearVelScale=<0.3 0.5 0.3>
    simSpaceSettings {
     maxAngularAcceleration=1000
     maxAngularVelocity=3
     maxLinearAcceleration=300
     maxLinearVelocity=700
     velocityScaleZ=0
     externalLinearDragV=<0 0.5 0>
    }
   }
   skydive_Surfing {simSpaceSettings {
     maxAngularAcceleration=1000
     maxAngularVelocity=3
     maxLinearAcceleration=300
     maxLinearVelocity=700
     velocityScaleZ=0
     externalLinearDragV=<0 0.5 0>
    }}
   swimming {simSpaceSettings {
     maxAngularAcceleration=1000
     maxAngularVelocity=3
     maxLinearAcceleration=300
     maxLinearVelocity=700
     velocityScaleZ=0
     externalLinearDragV=<0 0.5 0>
    }}
   tactical_Sprint {simSpaceSettings {
     maxAngularAcceleration=1000
     maxAngularVelocity=3
     maxLinearAcceleration=300
     maxLinearVelocity=700
     velocityScaleZ=0
     externalLinearDragV=<0 0.5 0>
    }}
   ziplining {simSpaceSettings {
     maxAngularAcceleration=1000
     maxAngularVelocity=3
     maxLinearAcceleration=300
     maxLinearVelocity=700
     velocityScaleZ=0
     externalLinearDragV=<0 0.5 0>
    }}
  }
 }
 RBAN Settings Out Spine {
  type=RBANSettingStruct
  default {
   componentLinearAccScale=<0 0 0>
   componentLinearVelScale=<0 0 0>
   simSpaceSettings {
    dampingAlpha=1
    maxAngularAcceleration=1000
    maxAngularVelocity=3
    maxLinearAcceleration=300
    maxLinearVelocity=700
    velocityScaleZ=0
    worldAlpha=1
    externalAngularVelocity=<0 0 0>
    externalLinearDragV=<0 0.5 0>
    externalLinearVelocity=<0 0 0>
   }
  }
 }
 RBAN Settings Out_chain {
  type=RBANSettingStruct
  default {
   componentLinearAccScale=<0 0 0>
   componentLinearVelScale=<0 0 0>
   simSpaceSettings {
    dampingAlpha=1
    maxAngularAcceleration=1000
    maxAngularVelocity=3
    maxLinearAcceleration=300
    maxLinearVelocity=700
    velocityScaleZ=0
    worldAlpha=1
    externalAngularVelocity=<0 0 0>
    externalLinearDragV=<0 0.5 0>
    externalLinearVelocity=<0 0 0>
   }
  }
 }
 RbanOrCloth {type=boolean}
 resetClothLandingFromZilpine {type=boolean}
 RotationBP {type=Rotator}
 Sim Space Settings {
  type=SimSpaceSettings
  default {
   dampingAlpha=1
   maxAngularAcceleration=10000
   maxAngularVelocity=10000
   maxLinearAcceleration=10000
   maxLinearVelocity=10000
   velocityScaleZ=1
   worldAlpha=0
   externalLinearDragV=<0 0 0>
  }
 }
 Sim Space Settings_Base {
  type=SimSpaceSettings
  default {
   dampingAlpha=1
   maxAngularAcceleration=500
   maxAngularVelocity=3
   maxLinearAcceleration=700
   maxLinearVelocity=500
   velocityScaleZ=0
   worldAlpha=1
   externalLinearDragV=<1 1 0>
  }
 }
 Sim Space Settings_GrindingWeaponActive {
  type=SimSpaceSettings
  default {
   dampingAlpha=1
   maxAngularAcceleration=500
   maxAngularVelocity=3
   maxLinearAcceleration=700
   maxLinearVelocity=500
   velocityScaleZ=0
   worldAlpha=0.2
   externalLinearDragV=<1 1 0>
  }
 }
 Sim Space Settings_GrindTarget {
  type=SimSpaceSettings
  default {
   dampingAlpha=1
   maxAngularAcceleration=500
   maxAngularVelocity=3
   maxLinearAcceleration=700
   maxLinearVelocity=500
   velocityScaleZ=0
   worldAlpha=0.1
   externalLinearDragV=<1 1 0>
  }
 }
 SkydivingAlpha {type=Rotator}
 spin02 {type=Transform, default {location=<0 0 0>, rotation=<rot 0 0 0>, scale=<1 1 1>}}
 spin03 {type=Transform, default {location=<0 0 0>, rotation=<rot 0 0 0>, scale=<1 1 1>}}
 spin04 {type=Transform, default {location=<0 0 0>, rotation=<rot 0 0 0>, scale=<1 1 1>}}
 TeleportDistanceThreshold_Local {type=Vector}
 Wind Gust Params {
  type=WindGustParamsStruct
  default {
   dBNO {
    maxDownInterpSpeed=5
    maxGustDuration=2
    maxGustStrength=7
    maxTimeBetweenGusts=7
    maxUpInterpSpeed=5
    minDownInterpSpeed=2
    minGustDuration=0.5
    minGustStrength=3
    minTimeBetweenGusts=0
    minUpInterpSpeed=2
    windTurbulenceScalarA=0.1
    windTurbulenceScalarB=1
    windAmplitude=<0 0 0>
    windFrequency=<0 0 0>
    windOffset=<0 0 0>
   }
   dBNO_Carried {
    maxDownInterpSpeed=5
    maxGustDuration=2
    maxGustStrength=7
    maxTimeBetweenGusts=7
    maxUpInterpSpeed=5
    minDownInterpSpeed=2
    minGustDuration=0.5
    minGustStrength=3
    minTimeBetweenGusts=0
    minUpInterpSpeed=2
    windTurbulenceScalarA=0.1
    windTurbulenceScalarB=1
    windAmplitude=<0 0 0>
    windFrequency=<0 0 0>
    windOffset=<0 0 0>
   }
   dBNO_Impulse {
    maxDownInterpSpeed=5
    maxGustDuration=2
    maxGustStrength=7
    maxTimeBetweenGusts=7
    maxUpInterpSpeed=5
    minDownInterpSpeed=2
    minGustDuration=0.5
    minGustStrength=3
    minTimeBetweenGusts=0
    minUpInterpSpeed=2
    windTurbulenceScalarA=0.1
    windTurbulenceScalarB=1
    windAmplitude=<0 0 0>
    windFrequency=<0 0 0>
    windOffset=<0 0 0>
   }
   dBNO_RollBwd {
    maxDownInterpSpeed=5
    maxGustDuration=2
    maxGustStrength=7
    maxTimeBetweenGusts=7
    maxUpInterpSpeed=5
    minDownInterpSpeed=2
    minGustDuration=0.5
    minGustStrength=3
    minTimeBetweenGusts=0
    minUpInterpSpeed=2
    windTurbulenceScalarA=0.1
    windTurbulenceScalarB=1
    windAmplitude=<0 0 0>
    windFrequency=<0 0 0>
    windOffset=<0 0 0>
   }
   dBNO_RollFwd {
    maxDownInterpSpeed=5
    maxGustDuration=2
    maxGustStrength=7
    maxTimeBetweenGusts=7
    maxUpInterpSpeed=5
    minDownInterpSpeed=2
    minGustDuration=0.5
    minGustStrength=3
    minTimeBetweenGusts=0
    minUpInterpSpeed=2
    windTurbulenceScalarA=0.1
    windTurbulenceScalarB=1
    windAmplitude=<0 0 0>
    windFrequency=<0 0 0>
    windOffset=<0 0 0>
   }
   dBNO_Tumbling {
    maxDownInterpSpeed=5
    maxGustDuration=2
    maxGustStrength=7
    maxTimeBetweenGusts=7
    maxUpInterpSpeed=5
    minDownInterpSpeed=2
    minGustDuration=0.5
    minGustStrength=3
    minTimeBetweenGusts=0
    minUpInterpSpeed=2
    windTurbulenceScalarA=0.1
    windTurbulenceScalarB=1
    windAmplitude=<0 0 0>
    windFrequency=<0 0 0>
    windOffset=<0 0 0>
   }
   emoteOrMelee {
    maxDownInterpSpeed=5
    maxGustDuration=2
    maxGustStrength=7
    maxTimeBetweenGusts=7
    maxUpInterpSpeed=5
    minDownInterpSpeed=2
    minGustDuration=0.5
    minGustStrength=3
    minTimeBetweenGusts=0
    minUpInterpSpeed=2
    windTurbulenceScalarA=0.1
    windTurbulenceScalarB=1
    windAmplitude=<30 25 50>
    windFrequency=<2 3 5>
    windOffset=<-300 -150 0>
   }
   front_End {
    maxDownInterpSpeed=5
    maxGustDuration=2
    maxGustStrength=10
    maxTimeBetweenGusts=7
    maxUpInterpSpeed=5
    minDownInterpSpeed=2
    minGustDuration=0.5
    minGustStrength=5
    minTimeBetweenGusts=0
    minUpInterpSpeed=2
    windTurbulenceScalarA=0.1
    windTurbulenceScalarB=1
    windAmplitude=<10 30 50>
    windFrequency=<15 6 7>
    windOffset=<0 -200 -250>
   }
   onGround_Crouch_Moving {
    maxDownInterpSpeed=5
    maxGustDuration=2
    maxGustStrength=7
    maxTimeBetweenGusts=7
    maxUpInterpSpeed=5
    minDownInterpSpeed=2
    minGustDuration=0.5
    minGustStrength=3
    minTimeBetweenGusts=0
    minUpInterpSpeed=2
    windTurbulenceScalarA=0.1
    windTurbulenceScalarB=1
    windAmplitude=<100 25 75>
    windFrequency=<1.5 0.5 7.5>
    windOffset=<0 -100 -175>
   }
   onGround_Falling {
    maxDownInterpSpeed=5
    maxGustDuration=8
    maxGustStrength=7
    maxTimeBetweenGusts=8
    maxUpInterpSpeed=5
    minDownInterpSpeed=2
    minGustDuration=0.2
    minGustStrength=3
    minTimeBetweenGusts=1
    minUpInterpSpeed=2
    windTurbulenceScalarA=0.1
    windTurbulenceScalarB=0.8
    windAmplitude=<20 65 42>
    windFrequency=<5 6 8.6>
    windOffset=<30 -100 -300>
   }
   onGround_Moving {
    maxDownInterpSpeed=5
    maxGustDuration=4
    maxGustStrength=4
    maxTimeBetweenGusts=8
    maxUpInterpSpeed=5
    minDownInterpSpeed=2
    minGustDuration=0.2
    minGustStrength=3
    minTimeBetweenGusts=1
    minUpInterpSpeed=2
    windTurbulenceScalarA=0.1
    windTurbulenceScalarB=0.3
    windAmplitude=<10 15 28>
    windFrequency=<1.5 2.3 3.7>
    windOffset=<200 -500 -1000>
   }
   onGround_Standing {
    maxDownInterpSpeed=5
    maxGustDuration=8
    maxGustStrength=7
    maxTimeBetweenGusts=8
    maxUpInterpSpeed=5
    minDownInterpSpeed=2
    minGustDuration=0.2
    minGustStrength=2
    minTimeBetweenGusts=1
    minUpInterpSpeed=2
    windTurbulenceScalarA=0.1
    windTurbulenceScalarB=0.8
    windAmplitude=<35 15 28>
    windFrequency=<1.5 2.3 3.7>
    windOffset=<50 -30 -150>
   }
   rail_Grinding {
    maxDownInterpSpeed=5
    maxGustDuration=8
    maxGustStrength=7
    maxTimeBetweenGusts=8
    maxUpInterpSpeed=5
    minDownInterpSpeed=2
    minGustDuration=0.2
    minGustStrength=3
    minTimeBetweenGusts=1
    minUpInterpSpeed=2
    windTurbulenceScalarA=0.1
    windTurbulenceScalarB=0.8
    windAmplitude=<20 65 42>
    windFrequency=<5 6 8.6>
    windOffset=<50 -500 -400>
   }
   riding_Moving {
    maxDownInterpSpeed=5
    maxGustDuration=2
    maxGustStrength=7
    maxTimeBetweenGusts=7
    maxUpInterpSpeed=5
    minDownInterpSpeed=2
    minGustDuration=0.5
    minGustStrength=3
    minTimeBetweenGusts=0
    minUpInterpSpeed=2
    windTurbulenceScalarA=0.1
    windTurbulenceScalarB=1
    windAmplitude=<0 0 0>
    windFrequency=<0 0 0>
    windOffset=<0 0 0>
   }
   riding_Moving_Targeting {
    maxDownInterpSpeed=5
    maxGustDuration=2
    maxGustStrength=7
    maxTimeBetweenGusts=7
    maxUpInterpSpeed=5
    minDownInterpSpeed=2
    minGustDuration=0.5
    minGustStrength=3
    minTimeBetweenGusts=0
    minUpInterpSpeed=2
    windTurbulenceScalarA=0.1
    windTurbulenceScalarB=1
    windAmplitude=<0 0 0>
    windFrequency=<0 0 0>
    windOffset=<0 0 0>
   }
   riding_Standing {
    maxDownInterpSpeed=5
    maxGustDuration=2
    maxGustStrength=7
    maxTimeBetweenGusts=7
    maxUpInterpSpeed=5
    minDownInterpSpeed=2
    minGustDuration=0.5
    minGustStrength=3
    minTimeBetweenGusts=0
    minUpInterpSpeed=2
    windTurbulenceScalarA=0.1
    windTurbulenceScalarB=1
    windAmplitude=<0 0 0>
    windFrequency=<0 0 0>
    windOffset=<0 0 0>
   }
   skydive_DivingDown {
    maxDownInterpSpeed=5
    maxGustDuration=8
    maxGustStrength=7
    maxTimeBetweenGusts=8
    maxUpInterpSpeed=5
    minDownInterpSpeed=2
    minGustDuration=0.2
    minGustStrength=3
    minTimeBetweenGusts=1
    minUpInterpSpeed=2
    windTurbulenceScalarA=0.1
    windTurbulenceScalarB=4
    windAmplitude=<0 25 35>
    windFrequency=<0 4 10>
    windOffset=<-450 -100 100>
   }
   skydive_FloatingDown {
    maxDownInterpSpeed=5
    maxGustDuration=8
    maxGustStrength=1
    maxTimeBetweenGusts=8
    maxUpInterpSpeed=5
    minDownInterpSpeed=2
    minGustDuration=0.2
    minGustStrength=0
    minTimeBetweenGusts=1
    minUpInterpSpeed=2
    windTurbulenceScalarA=0.5
    windTurbulenceScalarB=1
    windAmplitude=<50 20 35>
    windFrequency=<8 2 3>
    windOffset=<0 -300 200>
   }
   skydive_GoingUp {
    maxDownInterpSpeed=5
    maxGustDuration=8
    maxGustStrength=7
    maxTimeBetweenGusts=8
    maxUpInterpSpeed=5
    minDownInterpSpeed=2
    minGustDuration=0.2
    minGustStrength=3
    minTimeBetweenGusts=1
    minUpInterpSpeed=2
    windTurbulenceScalarA=0.1
    windTurbulenceScalarB=0.8
    windAmplitude=<0 0 0>
    windFrequency=<0 0 0>
    windOffset=<0 -500 300>
   }
   skydive_Parachuting {
    maxDownInterpSpeed=5
    maxGustDuration=8
    maxGustStrength=1
    maxTimeBetweenGusts=8
    maxUpInterpSpeed=5
    minDownInterpSpeed=2
    minGustDuration=5
    minGustStrength=0
    minTimeBetweenGusts=5
    minUpInterpSpeed=2
    windTurbulenceScalarA=0.2
    windTurbulenceScalarB=1
    windAmplitude=<25 50 75>
    windFrequency=<8 6 2>
    windOffset=<0 -350 -275>
   }
   skydive_Surfing {
    maxDownInterpSpeed=5
    maxGustDuration=8
    maxGustStrength=1
    maxTimeBetweenGusts=16
    maxUpInterpSpeed=5
    minDownInterpSpeed=2
    minGustDuration=5
    minGustStrength=0
    minTimeBetweenGusts=5
    minUpInterpSpeed=2
    windTurbulenceScalarA=0.2
    windTurbulenceScalarB=1
    windAmplitude=<25 50 75>
    windFrequency=<8 6 2>
    windOffset=<250 -200 -200>
   }
   swimming {
    maxDownInterpSpeed=5
    maxGustDuration=2
    maxGustStrength=7
    maxTimeBetweenGusts=7
    maxUpInterpSpeed=5
    minDownInterpSpeed=2
    minGustDuration=0.5
    minGustStrength=3
    minTimeBetweenGusts=0
    minUpInterpSpeed=2
    windTurbulenceScalarA=0.1
    windTurbulenceScalarB=1
    windAmplitude=<10 25 20>
    windFrequency=<1.5 0.5 7.5>
    windOffset=<0 -150 50>
   }
   tactical_Sprint {
    maxDownInterpSpeed=5
    maxGustDuration=2
    maxGustStrength=7
    maxTimeBetweenGusts=7
    maxUpInterpSpeed=5
    minDownInterpSpeed=2
    minGustDuration=0.5
    minGustStrength=3
    minTimeBetweenGusts=0
    minUpInterpSpeed=2
    windTurbulenceScalarA=0.1
    windTurbulenceScalarB=1
    windAmplitude=<0 0 0>
    windFrequency=<0 0 0>
    windOffset=<0 0 0>
   }
   ziplining {
    maxDownInterpSpeed=5
    maxGustDuration=8
    maxGustStrength=7
    maxTimeBetweenGusts=8
    maxUpInterpSpeed=5
    minDownInterpSpeed=2
    minGustDuration=0.2
    minGustStrength=3
    minTimeBetweenGusts=1
    minUpInterpSpeed=2
    windTurbulenceScalarA=0.1
    windTurbulenceScalarB=0.8
    windAmplitude=<20 65 42>
    windFrequency=<5 6 8.6>
    windOffset=<30 -100 -300>
   }
  }
 }
]
