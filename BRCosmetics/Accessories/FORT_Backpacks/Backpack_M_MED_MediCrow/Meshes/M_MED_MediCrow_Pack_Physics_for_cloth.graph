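# Physics asset for the MediCrow backpack cosmetic, named as the collision set for its cloth.
# Each body is a single capsule: three tapering along the cape spine bones, and mirrored pairs on joints 3 and 5 of the left and right chains.
# A much larger offset capsule sits on the `hider` bone.
# There are no constraints, so the asset defines collision volumes only, with no jointed ragdoll.
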PhysicsAsset /BRCosmetics/Accessories/FORT_Backpacks/Backpack_M_MED_MediCrow/Meshes/M_MED_MediCrow_Pack_Physics_for_cloth
bodies {
 dyn_cape_spine1 [{
   shapeType=Capsule
   length=11.6658
   radius=6.5807
   shapeName=dyn_cape_spine1_capsule
   center=<-1.9823 -0.0002 0.6837>
   rotation=<rot 0 0.0048 -89.9962>
  }]
 dyn_cape_spine2 [{shapeType=Capsule, length=8.9324, radius=3.9555, shapeName=dyn_cape_spine2_capsule, center=<0.1505 0 0.2544>, rotation=<rot 0 179.9999 -90>}]
 dyn_cape_spine3 [{
   shapeType=Capsule
   length=6.5071
   radius=2.2893
   shapeName=dyn_cape_spine3_capsule
   center=<1.8981 -0.0031 0.678>
   rotation=<rot 0.7618 88.9006 -73.248>
  }]
 dyn_l_newchain_joint3 [{
   shapeType=Capsule
   length=10.3268
   radius=6
   shapeName=dyn_l_newchain_joint3_capsule
   center=<3.6 -3.7827 -0.0315>
   rotation=<rot -3.8542 -87.3914 -86.7568>
  }]
 dyn_l_newchain_joint5 [{
   shapeType=Capsule
   length=4.9221
   radius=5.5
   shapeName=dyn_l_newchain_joint5_capsule
   center=<4.243 -3.9462 0.5406>
   rotation=<rot -0.0002 95.5329 -108.4617>
  }]
 dyn_r_newchain_joint3 [{
   shapeType=Capsule
   length=10.3269
   radius=6
   shapeName=dyn_r_newchain_joint3_capsule
   center=<3.6 -4.9036 0.0315>
   rotation=<rot -3.8542 92.6084 -86.7566>
  }]
 dyn_r_newchain_joint5 [{
   shapeType=Capsule
   length=4.9221
   radius=5.5
   shapeName=dyn_r_newchain_joint5_capsule
   center=<4.81 -3.8913 -0.5513>
   rotation=<rot -0.0002 -84.4672 -108.4617>
  }]
 hider [{shapeType=Capsule, length=18.6363, radius=15.0566, center=<-0.3106 -17.2648 -8.7714>, rotation=<rot 9.1235 -90 90.8285>}]
}
constraints []
